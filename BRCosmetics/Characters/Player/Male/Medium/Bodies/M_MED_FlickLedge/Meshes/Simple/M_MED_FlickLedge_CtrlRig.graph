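# Control Rig for the Flick Ledge male player body that deforms its robe bones to follow the legs.
# Robe Deformation aims deform thigh and calf bones, then constrains and aims robe bones, blending rear chains from thigh to calf by curve.
# It also outputs knee bend, spherical pose reader thigh-raise values and Kalman-smoothed foot-crossing values.
# The main graph feeds knee bend into mirrored Modify Transforms functions as weights for corrective offsets.
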ControlRigBlueprint /BRCosmetics/Characters/Player/Male/Medium/Bodies/M_MED_FlickLedge/Meshes/Simple/M_MED_FlickLedge_CtrlRig
graphs [
 New Function_ContainedGraph {nodes [
   RigVMFunctionEntryNode()
   > RigVMFunction_Sequence = RigVMAggregateNode()
   &RigUnit_SetRotation(
    Item=(Type=Bone,Name="CR_root_no_rotation")
    Space=GlobalSpace
    Value=(X=0,Y=0,Z=0,W=1)
    Weight=1
    bInitial=False
    ExecutePin=@RigVMFunction_Sequence.A
   )
   > RigVMFunction_Sequence_2 = RigVMAggregateNode()
   &RigUnit_ParentConstraint(
    AdvancedSettings=(InterpolationType=Average,RotationOrderForFilter=XZY)
    Child=(Type=Bone,Name="dyn_robe_side_l_02")
    Parents=((Item=(Type=Bone,Name="calf_l"),Weight=1))
    Weight=1
    ExecutePin=@RigVMFunction_Sequence_2.D
   )
   > &RigUnit_ParentConstraint(
    AdvancedSettings=(InterpolationType=Average,RotationOrderForFilter=XZY)
    Child=(Type=Bone,Name="dyn_robe_side_r_02")
    Parents=((Item=(Type=Bone,Name="calf_r"),Weight=1))
    Weight=1
   )
   n1 = /Script/ControlRig.RigUnit_AimBoneMath(
    DebugSettings=(bEnabled=False,Scale=10,WorldOffset=(Rotation=(X=0,Y=0,Z=0,W=1),Translation=(X=0,Y=0,Z=0),Scale3D=(X=1,Y=1,Z=1)))
    Primary=(Weight=1,Axis=(X=1,Y=0,Z=0),Target=(X=-1,Y=0,Z=0),Kind=Location,Space=(Type=Bone,Name="CR_root_no_rotation"))
    Secondary=(Weight=1,Axis=(X=0,Y=0,Z=-1),Target=(X=1,Y=0,Z=0),Kind=Location,Space=(Type=Bone,Name="CR_root_no_rotation"))
    Weight=1
    Secondary.Target=&RigUnit_GetTransform(Item=(Type=Bone,Name="calf_l"), bInitial=False).Transform.Translation
    InputTransform=&RigUnit_GetTransform(Item=(Type=Bone,Name="deform_calf_r"), bInitial=False).Transform
    Primary.Target=&RigUnit_GetTransform(Item=(Type=Bone,Name="foot_r"), bInitial=False).Transform.Translation
   ) {is=struct RigVMUnitNode}
   n2 = &RigUnit_SetRotation(
    Item=(Type=Bone,Name="deform_calf_r")
    Space=GlobalSpace
    Weight=1
    bInitial=False
    Value=@n1.Result.Rotation
    ExecutePin=@RigVMFunction_Sequence_2.C
   )
   n3 = &RigVMFunction_Sequence(ExecuteContext=@n2.ExecutePin)
   ItemArray_4 = RigVMDispatchNode(Value=((Type=Bone,Name="dyn_robe_rear_r_01"),(Type=Bone,Name="dyn_robe_rear_r_02"),(Type=Bone,Name="dyn_robe_rear_r_03"),(Type=Bone,Name="dyn_robe_rear_r_04"),(Type=Bone,Name="dyn_robe_rear_r_05"))) {template="DISPATCH_RigVMDispatch_Constant(io Value)"}
   For_Each_4 = RigVMDispatchNode(Array=@ItemArray_4.Value, ExecuteContext=@n3.A) {template="DISPATCH_RigVMDispatch_ArrayIterator(in Array,out Element,out Index,out Count,out Ratio)"}
   n4 = &RigVMFunction_AnimEvalRichCurve(
    Curve=(EditorCurveData=(Keys=((),(Time=0.3048,Value=0.535),(Time=1.0018,Value=0.8752))))
    SourceMinimum=0
    TargetMinimum=0
    Value=@For_Each_4.Ratio
   )
   > n5 = &RigUnit_ParentConstraint(
    AdvancedSettings=(InterpolationType=Average,RotationOrderForFilter=XZY)
    Filter=(TranslationFilter=(bX=True,bY=True,bZ=True),RotationFilter=(bX=false,bY=false,bZ=false),ScaleFilter=(bX=True,bY=True,bZ=True))
    Parents=((Item=(Type=Bone,Name="deform_thigh_r"),Weight=1),(Item=(Type=Bone,Name="calf_r"),Weight=1))
    Weight=1
    Child=@For_Each_4.Element
    Parents.0.Weight=&RigVMFunction_MathDoubleSub(A=1, B=@n4.Result).Result
    Parents.1.Weight=@n4.Result
   )
   n6 = &RigVMFunction_ControlFlowBranch(
    Condition=RigVMDispatchNode(B=0, A=@For_Each_4.Index) {template="DISPATCH_RigVMDispatch_CoreEquals(in A,in B,out Result)"}.Result
    ExecuteContext=@n5.ExecutePin
   )
   n7 = &RigUnit_GetTransform(bInitial=false, Item=@For_Each_4.Element)
   n8 = &RigVMFunction_MathTransformMakeRelative(
    Global=&RigUnit_GetTransform(Item=(Type=Bone,Name="dyn_robe_rear_r_05"), bInitial=true).Transform
    Parent=&RigUnit_GetTransform(Item=(Type=Bone,Name="deform_calf_r"), bInitial=true).Transform
   )
   n9 = &RigUnit_AimBoneMath(
    DebugSettings=(bEnabled=False,Scale=10,WorldOffset=(Rotation=(X=0,Y=0,Z=0,W=1),Translation=(X=0,Y=0,Z=0),Scale3D=(X=1,Y=1,Z=1)))
    Primary=(Weight=1,Axis=(X=-1,Y=0,Z=0),Target=(X=-1,Y=0,Z=0),Kind=Location,Space=(Type=Bone,Name="CR_root_no_rotation"))
    Secondary=(Weight=0,Axis=(X=0,Y=0,Z=1),Target=(X=1,Y=0,Z=0),Kind=Location,Space=(Type=Bone,Name="CR_root_no_rotation"))
    Weight=1
    InputTransform=@n7.Transform
    Primary.Target=/Script/RigVM.RigVMFunction_MathTransformTransformVector(
     Location=(X=-100,Y=0,Z=0)
     Transform=&RigVMFunction_MathTransformMul(A=@n8.Local, B=@n1.Result) {template="Multiply::Execute(in A,in B,out Result)"}.Result
    ) {is=struct RigVMUnitNode}.Result
   )
   &RigUnit_OffsetTransformForItem(
    OffsetTransform=(Rotation=(X=0,Y=0,Z=0,W=1),Translation=(X=0,Y=0,Z=0),Scale3D=(X=1,Y=1,Z=1))
    Weight=1
    bPropagateToChildren=true
    OffsetTransform.Rotation=&RigVMFunction_MathTransformMakeRelative(Global=@n9.Result, Parent=@n7.Transform).Local.Rotation
    ExecutePin=@n6.False
    Item=@For_Each_4.Element
   )
   n10 = &RigVMFunction_MathTransformMakeRelative(
    Global=&RigUnit_GetTransform(Item=(Type=Bone,Name="dyn_robe_front_r_05"), bInitial=true).Transform
    Parent=&RigUnit_GetTransform(Item=(Type=Bone,Name="deform_calf_r"), bInitial=true).Transform
   )
   n11 = &RigUnit_AimBoneMath(
    DebugSettings=(bEnabled=False,Scale=10,WorldOffset=(Rotation=(X=0,Y=0,Z=0,W=1),Translation=(X=0,Y=0,Z=0),Scale3D=(X=1,Y=1,Z=1)))
    Primary=(Weight=1,Axis=(X=-1,Y=0,Z=0),Target=(X=1,Y=0,Z=0),Kind=Location,Space=(Type=Bone,Name="CR_root_no_rotation"))
    Secondary=(Weight=0,Axis=(X=0,Y=0,Z=-1),Target=(X=1,Y=0,Z=0),Kind=Location,Space=(Type=Bone,Name="CR_root_no_rotation"))
    Weight=1
    InputTransform=&RigUnit_GetTransform(Item=(Type=Bone,Name="dyn_robe_front_r_02"), bInitial=false).Transform
    Primary.Target=&RigVMFunction_MathTransformTransformVector(
     Location=(X=-20,Y=0,Z=0)
     Transform=&RigVMFunction_MathTransformMul(A=@n10.Local, B=@n1.Result) {template="Multiply::Execute(in A,in B,out Result)"}.Result
    ).Result
   )
   &RigUnit_SetRotation(Item=(Type=Bone,Name="dyn_robe_front_r_02"), Space=GlobalSpace, Weight=1, bInitial=False, Value=@n11.Result.Rotation, ExecutePin=@n3.B)
   > &RigUnit_ParentConstraint(
    AdvancedSettings=(InterpolationType=Average,RotationOrderForFilter=XZY)
    Child=(Type=Bone,Name="dyn_robe_front_r_02")
    Filter=(TranslationFilter=(bX=True,bY=True,bZ=True),RotationFilter=(bX=false,bY=false,bZ=false),ScaleFilter=(bX=True,bY=True,bZ=True))
    Parents=((Item=(Type=Bone,Name="calf_r"),Weight=1))
    Weight=1
   )
   RigUnit_GetTransform = &RigUnit_GetTransform(Item=(Type=Bone,Name="deform_thigh_l"), bInitial=false)
   n12 = &RigUnit_AimBoneMath(
    DebugSettings=(bEnabled=False,Scale=10,WorldOffset=(Rotation=(X=0,Y=0,Z=0,W=1),Translation=(X=0,Y=0,Z=0),Scale3D=(X=1,Y=1,Z=1)))
    Primary=(Weight=1,Axis=(X=-1,Y=0,Z=0),Target=(X=-1,Y=0,Z=0),Kind=Location,Space=(Type=Bone,Name="calf_l"))
    Secondary=(Weight=1,Axis=(X=0,Y=0,Z=1),Target=(X=1,Y=0,Z=0),Kind=Location,Space=(Type=Bone,Name="thigh_r"))
    Weight=1
    InputTransform=@RigUnit_GetTransform.Transform
   )
   n13 = &RigUnit_GetTransform(Item=(Type=Bone,Name="thigh_l"), bInitial=False)
   &RigUnit_SetTransform(
    Item=(Type=Bone,Name="deform_thigh_l")
    Value=(Rotation=(X=0,Y=0,Z=0,W=1),Translation=(X=0,Y=0,Z=0),Scale3D=(X=1,Y=1,Z=1))
    Weight=1
    bInitial=false
    bPropagateToChildren=true
    Value.Rotation=@n12.Result.Rotation
    Value.Scale3D=@n13.Transform.Scale3D
    Value.Translation=@n13.Transform.Translation
    ExecutePin=@RigVMFunction_Sequence_2.A
   )
   n14 = &RigUnit_GetTransform(Item=(Type=Bone,Name="calf_l"), Space=LocalSpace, bInitial=false)
   > &RigUnit_SetTransform(
    Item=(Type=Bone,Name="deform_calf_l")
    Space=LocalSpace
    Value=(Rotation=(X=0,Y=0,Z=0,W=1),Translation=(X=0,Y=0,Z=0),Scale3D=(X=1,Y=1,Z=1))
    Weight=1
    bInitial=False
    bPropagateToChildren=true
    Value.Rotation=@n14.Transform.Rotation
    Value.Scale3D=@n14.Transform.Scale3D
    Value.Translation=@n14.Transform.Translation
   )
   n15 = &RigUnit_AimBoneMath(
    DebugSettings=(bEnabled=False,Scale=10,WorldOffset=(Rotation=(X=0,Y=0,Z=0,W=1),Translation=(X=0,Y=0,Z=0),Scale3D=(X=1,Y=1,Z=1)))
    Primary=(Weight=1,Axis=(X=1,Y=0,Z=0),Target=(X=1,Y=0,Z=0),Kind=Location,Space=(Type=Bone,Name="calf_r"))
    Secondary=(Weight=1,Axis=(X=0,Y=0,Z=-1),Target=(X=-1,Y=0,Z=0),Kind=Location,Space=(Type=Bone,Name="thigh_l"))
    Weight=1
    InputTransform=&RigUnit_GetTransform(Item=(Type=Bone,Name="deform_thigh_r"), bInitial=false).Transform
   )
   n16 = &RigUnit_GetTransform(Item=(Type=Bone,Name="thigh_r"), bInitial=False)
   > &RigUnit_SetTransform(
    Item=(Type=Bone,Name="deform_thigh_r")
    Value=(Rotation=(X=0,Y=0,Z=0,W=1),Translation=(X=0,Y=0,Z=0),Scale3D=(X=1,Y=1,Z=1))
    Weight=1
    bInitial=false
    bPropagateToChildren=true
    Value.Rotation=@n15.Result.Rotation
    Value.Scale3D=@n16.Transform.Scale3D
    Value.Translation=@n16.Transform.Translation
   )
   n17 = &RigUnit_GetTransform(Item=(Type=Bone,Name="calf_r"), Space=LocalSpace, bInitial=false)
   > &RigUnit_SetTransform(
    Item=(Type=Bone,Name="deform_calf_r")
    Space=LocalSpace
    Value=(Rotation=(X=0,Y=0,Z=0,W=1),Translation=(X=0,Y=0,Z=0),Scale3D=(X=1,Y=1,Z=1))
    Weight=1
    bInitial=False
    bPropagateToChildren=true
    Value.Rotation=@n17.Transform.Rotation
    Value.Scale3D=@n17.Transform.Scale3D
    Value.Translation=@n17.Transform.Translation
   )
   n18 = &RigUnit_AimBoneMath(
    DebugSettings=(bEnabled=False,Scale=10,WorldOffset=(Rotation=(X=0,Y=0,Z=0,W=1),Translation=(X=0,Y=0,Z=0),Scale3D=(X=1,Y=1,Z=1)))
    Primary=(Weight=1,Axis=(X=-1,Y=0,Z=0),Target=(X=-1,Y=0,Z=0),Kind=Location,Space=(Type=Bone,Name="CR_root_no_rotation"))
    Secondary=(Weight=1,Axis=(X=0,Y=0,Z=1),Target=(X=1,Y=0,Z=0),Kind=Location,Space=(Type=Bone,Name="CR_root_no_rotation"))
    Weight=1
    InputTransform=&RigUnit_GetTransform(Item=(Type=Bone,Name="deform_calf_l"), bInitial=False).Transform
    Primary.Target=&RigUnit_GetTransform(Item=(Type=Bone,Name="foot_l"), bInitial=False).Transform.Translation
    Secondary.Target=&RigUnit_GetTransform(Item=(Type=Bone,Name="calf_r"), bInitial=False).Transform.Translation
   )
   n19 = &RigUnit_SetRotation(
    Item=(Type=Bone,Name="deform_calf_l")
    Space=GlobalSpace
    Weight=1
    bInitial=False
    Value=@n18.Result.Rotation
    ExecutePin=@RigVMFunction_Sequence_2.B
   )
   n20 = &RigVMFunction_Sequence(ExecuteContext=@n19.ExecutePin)
   n21 = &RigVMFunction_MathTransformMakeRelative(
    Global=&RigUnit_GetTransform(Item=(Type=Bone,Name="dyn_robe_front_l_05"), bInitial=true).Transform
    Parent=&RigUnit_GetTransform(Item=(Type=Bone,Name="deform_calf_l"), bInitial=true).Transform
   )
   n22 = &RigUnit_AimBoneMath(
    DebugSettings=(bEnabled=False,Scale=10,WorldOffset=(Rotation=(X=0,Y=0,Z=0,W=1),Translation=(X=0,Y=0,Z=0),Scale3D=(X=1,Y=1,Z=1)))
    Primary=(Weight=1,Axis=(X=1,Y=0,Z=0),Target=(X=-1,Y=0,Z=0),Kind=Location,Space=(Type=Bone,Name="CR_root_no_rotation"))
    Secondary=(Weight=0,Axis=(X=1,Y=0,Z=0),Target=(X=-1,Y=0,Z=0),Kind=Location,Space=(Type=Bone,Name="CR_root_no_rotation"))
    Weight=1
    InputTransform=&RigUnit_GetTransform(Item=(Type=Bone,Name="dyn_robe_front_l_02"), bInitial=false).Transform
    Primary.Target=&RigVMFunction_MathTransformTransformVector(
     Location=(X=20,Y=0,Z=0)
     Transform=&RigVMFunction_MathTransformMul(A=@n21.Local, B=@n18.Result) {template="Multiply::Execute(in A,in B,out Result)"}.Result
    ).Result
   )
   &RigUnit_SetRotation(Item=(Type=Bone,Name="dyn_robe_front_l_02"), Space=GlobalSpace, Weight=1, bInitial=False, Value=@n22.Result.Rotation, ExecutePin=@n20.B)
   > &RigUnit_ParentConstraint(
    AdvancedSettings=(InterpolationType=Average,RotationOrderForFilter=XZY)
    Child=(Type=Bone,Name="dyn_robe_front_l_02")
    Filter=(TranslationFilter=(bX=True,bY=True,bZ=True),RotationFilter=(bX=false,bY=false,bZ=false),ScaleFilter=(bX=True,bY=True,bZ=True))
    Parents=((Item=(Type=Bone,Name="calf_l"),Weight=1))
    Weight=1
   )
   ItemArray_3 = RigVMDispatchNode(Value=((Type=Bone,Name="dyn_robe_rear_l_01"),(Type=Bone,Name="dyn_robe_rear_l_02"),(Type=Bone,Name="dyn_robe_rear_l_03"),(Type=Bone,Name="dyn_robe_rear_l_04"),(Type=Bone,Name="dyn_robe_rear_l_05"))) {template="DISPATCH_RigVMDispatch_Constant(io Value)"}
   For_Each_2 = RigVMDispatchNode(Array=@ItemArray_3.Value, ExecuteContext=@n20.A) {template="DISPATCH_RigVMDispatch_ArrayIterator(in Array,out Element,out Index,out Count,out Ratio)"}
   RigVMFunction_AnimEvalRichCurve = &RigVMFunction_AnimEvalRichCurve(
    Curve=(EditorCurveData=(Keys=((),(Time=0.3048,Value=0.535),(Time=1.0018,Value=0.8752))))
    SourceMinimum=0
    TargetMinimum=0
    Value=@For_Each_2.Ratio
   )
   Subtract = &RigVMFunction_MathDoubleSub(A=1, B=@RigVMFunction_AnimEvalRichCurve.Result)
   > n23 = &RigUnit_ParentConstraint(
    AdvancedSettings=(InterpolationType=Average,RotationOrderForFilter=XZY)
    Filter=(TranslationFilter=(bX=True,bY=True,bZ=True),RotationFilter=(bX=false,bY=false,bZ=false),ScaleFilter=(bX=True,bY=True,bZ=True))
    Parents=((Item=(Type=Bone,Name="deform_thigh_l"),Weight=1),(Item=(Type=Bone,Name="calf_l"),Weight=1))
    Weight=1
    Child=@For_Each_2.Element
    Parents.0.Weight=@Subtract.Result
    Parents.1.Weight=@RigVMFunction_AnimEvalRichCurve.Result
   )
   n24 = &RigVMFunction_ControlFlowBranch(
    Condition=RigVMDispatchNode(B=0, A=@For_Each_2.Index) {template="DISPATCH_RigVMDispatch_CoreEquals(in A,in B,out Result)"}.Result
    ExecuteContext=@n23.ExecutePin
   )
   n25 = &RigUnit_GetTransform(bInitial=false, Item=@For_Each_2.Element)
   n26 = &RigUnit_GetTransform(Item=(Type=Bone,Name="dyn_robe_rear_l_05"), bInitial=true)
   n27 = &RigUnit_GetTransform(Item=(Type=Bone,Name="deform_calf_l"), bInitial=true)
   RigVMFunction_MathTransformTransformVector = &RigVMFunction_MathTransformTransformVector(
    Location=(X=100,Y=0,Z=0)
    Transform=&RigVMFunction_MathTransformMul(A=&RigVMFunction_MathTransformMakeRelative(Global=@n26.Transform, Parent=@n27.Transform).Local, B=@n18.Result) {template="Multiply::Execute(in A,in B,out Result)"}.Result
   )
   n28 = &RigUnit_AimBoneMath(
    DebugSettings=(bEnabled=False,Scale=10,WorldOffset=(Rotation=(X=0,Y=0,Z=0,W=1),Translation=(X=0,Y=0,Z=0),Scale3D=(X=1,Y=1,Z=1)))
    Primary=(Weight=1,Axis=(X=1,Y=0,Z=0),Target=(X=-1,Y=0,Z=0),Kind=Location,Space=(Type=Bone,Name="CR_root_no_rotation"))
    Secondary=(Weight=0,Axis=(X=0,Y=0,Z=1),Target=(X=1,Y=0,Z=0),Kind=Location,Space=(Type=Bone,Name="CR_root_no_rotation"))
    Weight=1
    InputTransform=@n25.Transform
    Primary.Target=@RigVMFunction_MathTransformTransformVector.Result
   )
   &RigUnit_OffsetTransformForItem(
    OffsetTransform=(Rotation=(X=0,Y=0,Z=0,W=1),Translation=(X=0,Y=0,Z=0),Scale3D=(X=1,Y=1,Z=1))
    Weight=1
    bPropagateToChildren=true
    OffsetTransform.Rotation=&RigVMFunction_MathTransformMakeRelative(Global=@n28.Result, Parent=@n25.Transform).Local.Rotation
    ExecutePin=@n24.False
    Item=@For_Each_2.Element
   )
   RigVMFunction_Sequence_1 = RigVMAggregateNode(ExecuteContext=@RigVMFunction_Sequence.B)
   Get Delta Ref = RigVMFunctionReferenceNode(Item=(Type=Bone,Name="calf_l"), ExecuteContext=@RigVMFunction_Sequence_1.A)
   > Get Delta Ref_1 = RigVMFunctionReferenceNode(Item=(Type=Bone,Name="calf_r"))
   n29 = &RigUnit_SphericalPoseReader(
    ActiveRegionScaleFactors=(PositiveWidth=3,NegativeWidth=1,PositiveHeight=1.2,NegativeHeight=1)
    ActiveRegionSize=0.1
    Debug=(bDrawDebug=false,bDraw2D=False,bDrawLocalAxes=False,DebugScale=25,DebugSegments=20,DebugThickness=0.25)
    DriverAxis=(X=-1,Y=0,Z=0)
    DriverItem=(Type=Bone,Name="thigh_l")
    FalloffRegionScaleFactors=(PositiveWidth=1,NegativeWidth=1,PositiveHeight=0.75,NegativeHeight=0.75)
    FalloffSize=0.58
    FlipHeightScaling=False
    FlipWidthScaling=False
    OptionalParentItem=(Type=Bone,Name="pelvis")
    RotationOffset=(X=90,Y=0,Z=-15)
    ExecutePin=@RigVMFunction_Sequence_1.B
   )
   > n30 = &RigUnit_SphericalPoseReader(
    ActiveRegionScaleFactors=(PositiveWidth=3,NegativeWidth=1,PositiveHeight=1,NegativeHeight=1)
    ActiveRegionSize=0.1
    Debug=(bDrawDebug=false,bDraw2D=False,bDrawLocalAxes=False,DebugScale=25,DebugSegments=20,DebugThickness=0.25)
    DriverAxis=(X=-1,Y=0,Z=0)
    DriverItem=(Type=Bone,Name="thigh_l")
    FalloffRegionScaleFactors=(PositiveWidth=1,NegativeWidth=1,PositiveHeight=1,NegativeHeight=1)
    FalloffSize=0.5
    FlipHeightScaling=False
    FlipWidthScaling=False
    OptionalParentItem=(Type=Bone,Name="pelvis")
    RotationOffset=(X=180,Y=0,Z=0)
   )
   > n31 = &RigUnit_SphericalPoseReader(
    ActiveRegionScaleFactors=(PositiveWidth=3,NegativeWidth=1,PositiveHeight=3,NegativeHeight=3)
    ActiveRegionSize=0.1
    Debug=(bDrawDebug=false,bDraw2D=False,bDrawLocalAxes=False,DebugScale=25,DebugSegments=20,DebugThickness=0.25)
    DriverAxis=(X=-1,Y=0,Z=0)
    DriverItem=(Type=Bone,Name="thigh_l")
    FalloffRegionScaleFactors=(PositiveWidth=1,NegativeWidth=0.95,PositiveHeight=1,NegativeHeight=1)
    FalloffSize=0.5
    FlipHeightScaling=False
    FlipWidthScaling=False
    OptionalParentItem=(Type=Bone,Name="pelvis")
    RotationOffset=(X=-90,Y=0,Z=0)
   )
   n32 = &RigUnit_SphericalPoseReader(
    ActiveRegionScaleFactors=(PositiveWidth=3,NegativeWidth=1,PositiveHeight=1.2,NegativeHeight=1)
    ActiveRegionSize=0.1
    Debug=(bDrawDebug=false,bDraw2D=False,bDrawLocalAxes=False,DebugScale=25,DebugSegments=20,DebugThickness=0.25)
    DriverAxis=(X=1,Y=0,Z=0)
    DriverItem=(Type=Bone,Name="thigh_r")
    FalloffRegionScaleFactors=(PositiveWidth=1,NegativeWidth=1,PositiveHeight=0.75,NegativeHeight=0.75)
    FalloffSize=0.58
    FlipHeightScaling=true
    FlipWidthScaling=true
    OptionalParentItem=(Type=Bone,Name="pelvis")
    RotationOffset=(X=-90,Y=0,Z=-15)
    ExecutePin=@RigVMFunction_Sequence_1.C
   )
   > n33 = &RigUnit_SphericalPoseReader(
    ActiveRegionScaleFactors=(PositiveWidth=3,NegativeWidth=1,PositiveHeight=1,NegativeHeight=1)
    ActiveRegionSize=0.1
    Debug=(bDrawDebug=false,bDraw2D=False,bDrawLocalAxes=False,DebugScale=25,DebugSegments=20,DebugThickness=0.25)
    DriverAxis=(X=1,Y=0,Z=0)
    DriverItem=(Type=Bone,Name="thigh_r")
    FalloffRegionScaleFactors=(PositiveWidth=1,NegativeWidth=1,PositiveHeight=1,NegativeHeight=1)
    FalloffSize=0.5
    FlipHeightScaling=true
    FlipWidthScaling=true
    OptionalParentItem=(Type=Bone,Name="pelvis")
    RotationOffset=(X=0,Y=0,Z=0)
   )
   > n34 = &RigUnit_SphericalPoseReader(
    ActiveRegionScaleFactors=(PositiveWidth=3,NegativeWidth=1,PositiveHeight=3,NegativeHeight=3)
    ActiveRegionSize=0.1
    Debug=(bDrawDebug=false,bDraw2D=False,bDrawLocalAxes=False,DebugScale=25,DebugSegments=20,DebugThickness=0.25)
    DriverAxis=(X=1,Y=0,Z=0)
    DriverItem=(Type=Bone,Name="thigh_r")
    FalloffRegionScaleFactors=(PositiveWidth=1,NegativeWidth=0.95,PositiveHeight=1,NegativeHeight=1)
    FalloffSize=0.5
    FlipHeightScaling=true
    FlipWidthScaling=true
    OptionalParentItem=(Type=Bone,Name="pelvis")
    RotationOffset=(X=90,Y=0,Z=0)
   )
   n35 = &RigUnit_GetRelativeTransformForItem(Child=(Type=Bone,Name="foot_l"), Parent=(Type=Bone,Name="root"), bChildInitial=False, bParentInitial=False)
   Greater = /Script/RigVM.RigVMFunction_MathDoubleGreater(B=0, A=@n35.RelativeTransform.Translation.Y) {is=struct RigVMUnitNode, template="Greater::Execute(in A,in B,out Result)"}
   Less = /Script/RigVM.RigVMFunction_MathDoubleLess(B=-5, A=@n35.RelativeTransform.Translation.X) {is=struct RigVMUnitNode, template="Less::Execute(in A,in B,out Result)"}
   RigVMFunction_MathBoolAnd = &RigVMFunction_MathBoolAnd(B=@Greater.Result, A=@Less.Result)
   KalmanFilter = &RigVMFunction_KalmanFloat(
    BufferSize=4
    Value=&RigVMFunction_MathDoubleRemap(
     SourceMaximum=-35
     SourceMinimum=-5
     TargetMaximum=1
     TargetMinimum=0
     Value=RigVMDispatchNode(False=0, True=@n35.RelativeTransform.Translation.X, Condition=@RigVMFunction_MathBoolAnd.Result) {template="DISPATCH_RigVMDispatch_If(in Condition,in True,in False,out Result)"}.Result
    ).Result
   ) {template="KalmanFilter::Execute(in Value,in BufferSize,out Result)"}
   n36 = &RigVMFunction_MathDoubleLess(B=-5, A=@n35.RelativeTransform.Translation.X) {template="Less::Execute(in A,in B,out Result)"}
   n37 = &RigVMFunction_MathDoubleLess(B=0, A=@n35.RelativeTransform.Translation.Y) {template="Less::Execute(in A,in B,out Result)"}
   n38 = &RigVMFunction_MathDoubleRemap(
    SourceMaximum=-20
    SourceMinimum=-10
    TargetMaximum=1
    TargetMinimum=0
    Value=RigVMDispatchNode(False=0, True=@n35.RelativeTransform.Translation.X, Condition=&RigVMFunction_MathBoolAnd(A=@n36.Result, B=@n37.Result).Result) {template="DISPATCH_RigVMDispatch_If(in Condition,in True,in False,out Result)"}.Result
   )
   RigUnit_GetRelativeTransformForItem = &RigUnit_GetRelativeTransformForItem(Child=(Type=Bone,Name="foot_r"), Parent=(Type=Bone,Name="root"), bChildInitial=False, bParentInitial=False)
   n39 = &RigVMFunction_MathDoubleGreater(B=0, A=@RigUnit_GetRelativeTransformForItem.RelativeTransform.Translation.Y) {template="Greater::Execute(in A,in B,out Result)"}
   n40 = &RigVMFunction_MathDoubleGreater(B=5, A=@RigUnit_GetRelativeTransformForItem.RelativeTransform.Translation.X) {template="Greater::Execute(in A,in B,out Result)"}
   n41 = &RigVMFunction_MathDoubleRemap(
    SourceMaximum=20
    SourceMinimum=5
    TargetMaximum=1
    TargetMinimum=0
    Value=RigVMDispatchNode(
     False=0
     True=@RigUnit_GetRelativeTransformForItem.RelativeTransform.Translation.X
     Condition=&RigVMFunction_MathBoolAnd(B=@n39.Result, A=@n40.Result).Result
    ) {template="DISPATCH_RigVMDispatch_If(in Condition,in True,in False,out Result)"}.Result
   )
   n42 = &RigVMFunction_MathDoubleGreater(B=5, A=@RigUnit_GetRelativeTransformForItem.RelativeTransform.Translation.X) {template="Greater::Execute(in A,in B,out Result)"}
   n43 = &RigVMFunction_MathDoubleLess(B=0, A=@RigUnit_GetRelativeTransformForItem.RelativeTransform.Translation.Y) {template="Less::Execute(in A,in B,out Result)"}
   n44 = &RigVMFunction_MathDoubleRemap(
    SourceMaximum=35
    SourceMinimum=5
    TargetMaximum=1
    TargetMinimum=0
    Value=RigVMDispatchNode(
     False=0
     True=@RigUnit_GetRelativeTransformForItem.RelativeTransform.Translation.X
     Condition=&RigVMFunction_MathBoolAnd(A=@n42.Result, B=@n43.Result).Result
    ) {template="DISPATCH_RigVMDispatch_If(in Condition,in True,in False,out Result)"}.Result
   )
   RigVMFunctionReturnNode(
    L Foot Across Front=@KalmanFilter.Result
    L Foot Across Rear=&RigVMFunction_KalmanFloat(BufferSize=4, Value=@n38.Result) {template="KalmanFilter::Execute(in Value,in BufferSize,out Result)"}.Result
    R Foot Across Front=&RigVMFunction_KalmanFloat(BufferSize=4, Value=@n41.Result) {template="KalmanFilter::Execute(in Value,in BufferSize,out Result)"}.Result
    R Foot Across Rear=&RigVMFunction_KalmanFloat(BufferSize=4, Value=@n44.Result) {template="KalmanFilter::Execute(in Value,in BufferSize,out Result)"}.Result
    L Knee Bend=&RigVMFunction_MathDoubleRemap(SourceMaximum=140, SourceMinimum=0, TargetMaximum=1, TargetMinimum=0, Value=@"Get Delta Ref".Result.Rotation.Yaw).Result
    R Knee Bend=&RigVMFunction_MathDoubleRemap(SourceMaximum=140, SourceMinimum=0, TargetMaximum=1, TargetMinimum=0, Value=@"Get Delta Ref_1".Result.Rotation.Yaw).Result
    L Thigh Raised Forward=@n29.OutputParam
    L Thigh Raised Side=@n30.OutputParam
    L Thigh Raised Rear=@n31.OutputParam
    R Thigh Raised Rear=@n34.OutputParam
    R Thigh Raised Side=@n33.OutputParam
    R Thigh Raised Forward=@n32.OutputParam
    ExecuteContext=@RigVMFunction_Sequence.C
   )
   &RigVMFunction_MathDoubleMul(B=1, A=@n38.Result)
   &RigVMFunction_MathDoubleMul(B=1, A=@n41.Result)
   &RigVMFunction_MathDoubleMul(B=1, A=@n44.Result)
  ]}
 New Function_ContainedGraph {nodes [
   n1 = RigVMFunctionEntryNode()
   IsEmpty = RigVMDispatchNode(Array=RigVMVariableNode() {variable=Deform Robe Chain}.Value) {template="DISPATCH_RigVMDispatch_ArrayIsEmpty(in Array,out IsEmpty)"}
   RigVMFunction_MathBoolNot = &RigVMFunction_MathBoolNot(Value=@IsEmpty.IsEmpty)
   RigVMFunction_ControlFlowBranch = &RigVMFunction_ControlFlowBranch(ExecuteContext=@n1.ExecuteContext, Condition=@RigVMFunction_MathBoolNot.Result)
   RigVMFunctionReturnNode(ExecuteContext=@RigVMFunction_ControlFlowBranch.Completed)
   RigVMFunction_Sequence_1 = RigVMAggregateNode(ExecuteContext=@RigVMFunction_ControlFlowBranch.True)
   n2 = RigVMVariableNode() {variable=Deform Robe Chain}
   n3 = RigVMDispatchNode(Index=0, Array=RigVMVariableNode() {variable=Deform Robe Chain}.Value) {template="DISPATCH_RigVMDispatch_ArrayGetAtIndex(in Array,in Index,out Element)"}
   /Script/ControlRig.RigUnit_DistributeRotationForItemArray(
    Items=((Type=Bone,Name="dyn_robe_rear_l_01"),(Type=Bone,Name="dyn_robe_rear_l_02"),(Type=Bone,Name="dyn_robe_rear_l_03"),(Type=Bone,Name="dyn_robe_rear_l_04"))
    RotationEaseType=Linear
    Rotations=((Rotation=(X=0,Y=0,Z=0,W=1),Ratio=0.7),(Rotation=(X=0,Y=0,Z=0,W=1),Ratio=1))
    Weight=1
    Items.0=RigVMDispatchNode(Index=3, Array=@n2.Value) {template="DISPATCH_RigVMDispatch_ArrayGetAtIndex(in Array,in Index,out Element)"}.Element
    Items.1=RigVMDispatchNode(Index=2, Array=@n2.Value) {template="DISPATCH_RigVMDispatch_ArrayGetAtIndex(in Array,in Index,out Element)"}.Element
    Items.2=RigVMDispatchNode(Index=1, Array=@n2.Value) {template="DISPATCH_RigVMDispatch_ArrayGetAtIndex(in Array,in Index,out Element)"}.Element
    Items.3=RigVMDispatchNode(Index=0, Array=@n2.Value) {template="DISPATCH_RigVMDispatch_ArrayGetAtIndex(in Array,in Index,out Element)"}.Element
    Rotations.1.Rotation=/Script/RigVM.RigVMFunction_MathQuaternionMakeRelative(
     Parent=&RigUnit_GetTransform(bInitial=false, Item=@n3.Element).Transform.Rotation
     Global=RigVMVariableNode() {variable=TargetAimQuat}.Value
    ) {is=struct RigVMUnitNode, template="Make Relative::Execute(in Global,in Parent,out Local)"}.Local
    ExecutePin=@RigVMFunction_Sequence_1.D
   ) {is=struct RigVMUnitNode}
   > For_Each_3 = RigVMDispatchNode(Array=RigVMVariableNode() {variable=Deform Robe Chain}.Value) {template="DISPATCH_RigVMDispatch_ArrayIterator(in Array,out Element,out Index,out Count,out Ratio)"}
   n4 = &RigVMFunction_ControlFlowBranch(
    Condition=RigVMDispatchNode(B=0, A=@For_Each_3.Index) {template="DISPATCH_RigVMDispatch_CoreEquals(in A,in B,out Result)"}.Result
    ExecuteContext=@For_Each_3.ExecuteContext
   )
   n5 = /Script/ControlRig.RigUnit_HierarchyGetParent(bDefaultParent=True, Child=RigVMVariableNode() {variable=Bind Foot}.Value) {is=struct RigVMUnitNode}
   &RigUnit_ParentConstraint(
    AdvancedSettings=(InterpolationType=Average,RotationOrderForFilter=XZY)
    Filter=(TranslationFilter=(bX=True,bY=True,bZ=True),RotationFilter=(bX=false,bY=false,bZ=false),ScaleFilter=(bX=false,bY=false,bZ=false))
    Parents=((Item=(Type=Bone,Name="None"),Weight=1),(Item=(Type=Bone,Name="None"),Weight=0.2),(Item=(Type=Bone,Name="None"),Weight=0.2))
    Weight=1
    Child=@For_Each_3.Element
    Parents.0.Item=@n5.Parent
    ExecutePin=@n4.False
   )
   &RigUnit_ParentConstraint(
    AdvancedSettings=(InterpolationType=Average,RotationOrderForFilter=XZY)
    Filter=(TranslationFilter=(bX=True,bY=True,bZ=True),RotationFilter=(bX=false,bY=false,bZ=false),ScaleFilter=(bX=false,bY=false,bZ=false))
    Parents=((Item=(Type=Bone,Name="None"),Weight=0.7),(Item=(Type=Bone,Name="None"),Weight=0.2),(Item=(Type=Bone,Name="None"),Weight=0.3))
    Weight=1
    Child=@For_Each_3.Element
    Parents.2.Item=@n5.Parent
    Parents.0.Item=&RigUnit_HierarchyGetParent(bDefaultParent=True, Child=@n5.Parent).Parent
    ExecutePin=@n4.True
   )
   RigVMDispatchNode(Array=RigVMVariableNode() {variable=TargetQuatArray}.Value, ExecuteContext=@RigVMFunction_Sequence_1.B) {template="DISPATCH_RigVMDispatch_ArrayReset(io Array)"}
   > For_Each_2 = RigVMDispatchNode(Array=RigVMVariableNode() {variable=Deform Robe Chain}.Value) {template="DISPATCH_RigVMDispatch_ArrayIterator(in Array,out Element,out Index,out Count,out Ratio)"}
   > RigVMDispatchNode(
    Element=&RigVMFunction_MathQuaternionMakeRelative(
     Parent=&RigUnit_GetTransform(bInitial=false, Item=@For_Each_2.Element).Transform.Rotation
     Global=RigVMVariableNode() {variable=TargetAimQuat}.Value
    ) {template="Make Relative::Execute(in Global,in Parent,out Local)"}.Local
    Array=RigVMVariableNode() {variable=TargetQuatArray}.Value
   ) {template="DISPATCH_RigVMDispatch_ArrayAdd(io Array,in Element,out Index)"}
   RigUnit_GetTransform = &RigUnit_GetTransform(bInitial=False, Item=RigVMVariableNode() {variable=Deform Calf}.Value)
   n6 = &RigUnit_AimBoneMath(
    DebugSettings=(bEnabled=False,Scale=10,WorldOffset=(Rotation=(X=0,Y=0,Z=0,W=1),Translation=(X=0,Y=0,Z=0),Scale3D=(X=1,Y=1,Z=1)))
    Primary=(Weight=1,Axis=(X=-1,Y=0,Z=0),Target=(X=-1,Y=0,Z=0),Kind=Location,Space=(Type=Bone,Name="CR_root_no_rotation"))
    Secondary=(Weight=1,Axis=(X=0,Y=0,Z=1),Target=(X=1,Y=0,Z=0),Kind=Location,Space=(Type=Bone,Name="CR_root_no_rotation"))
    Weight=1
    InputTransform=@RigUnit_GetTransform.Transform
    Primary.Target=&RigUnit_GetTransform(bInitial=False, Item=RigVMVariableNode() {variable=Bind Foot}.Value).Transform.Translation
    Secondary.Target=&RigUnit_GetTransform(bInitial=False, Item=RigVMVariableNode() {variable=Opposite Bind Calf}.Value).Transform.Translation
    Secondary.Axis=RigVMVariableNode() {variable=Aim Secondary Axis}.Value
    Primary.Axis=RigVMVariableNode() {variable=Aim Primary Axis}.Value
   )
   RigUnit_SetRotation = &RigUnit_SetRotation(
    Space=GlobalSpace
    Weight=1
    bInitial=False
    Value=@n6.Result.Rotation
    Item=RigVMVariableNode() {variable=Deform Calf}.Value
    ExecutePin=@RigVMFunction_Sequence_1.A
   )
   Multiply = /Script/RigVM.RigVMFunction_MathVectorMul(B=(X=-1,Y=-1,Z=-1), A=RigVMVariableNode() {variable=Aim Primary Axis}.Value) {is=struct RigVMUnitNode, template="Multiply::Execute(in A,in B,out Result)"}
   n7 = RigVMVariableNode() {variable=Deform Robe Chain}
   n8 = &RigUnit_GetTransform(
    bInitial=true
    Item=RigVMDispatchNode(Index=-1, Array=RigVMVariableNode() {variable=Deform Robe Chain}.Value) {template="DISPATCH_RigVMDispatch_ArrayGetAtIndex(in Array,in Index,out Element)"}.Element
   )
   n9 = &RigUnit_GetTransform(bInitial=true, Item=RigVMVariableNode() {variable=Deform Calf}.Value)
   RigVMFunction_MathTransformTransformVector = &RigVMFunction_MathTransformTransformVector(
    Transform=&RigVMFunction_MathTransformMul(A=&RigVMFunction_MathTransformMakeRelative(Global=@n8.Transform, Parent=@n9.Transform).Local, B=@n6.Result) {template="Multiply::Execute(in A,in B,out Result)"}.Result
    Location=&RigVMFunction_MathVectorMul(B=(X=-100,Y=-100,Z=-100), A=RigVMVariableNode() {variable=Aim Primary Axis}.Value) {template="Multiply::Execute(in A,in B,out Result)"}.Result
   )
   n10 = &RigUnit_AimBoneMath(
    DebugSettings=(bEnabled=False,Scale=10,WorldOffset=(Rotation=(X=0,Y=0,Z=0,W=1),Translation=(X=0,Y=0,Z=0),Scale3D=(X=1,Y=1,Z=1)))
    Primary=(Weight=1,Axis=(X=1,Y=0,Z=0),Target=(X=-1,Y=0,Z=0),Kind=Location,Space=(Type=Bone,Name="CR_root_no_rotation"))
    Secondary=(Weight=0,Axis=(X=0,Y=0,Z=1),Target=(X=1,Y=0,Z=0),Kind=Location,Space=(Type=Bone,Name="CR_root_no_rotation"))
    Weight=1
    Primary.Axis=@Multiply.Result
    InputTransform=&RigUnit_GetTransform(
     bInitial=false
     Item=RigVMDispatchNode(Index=0, Array=@n7.Value) {template="DISPATCH_RigVMDispatch_ArrayGetAtIndex(in Array,in Index,out Element)"}.Element
    ).Transform
    Secondary.Axis=RigVMVariableNode() {variable=Aim Secondary Axis}.Value
    Primary.Target=@RigVMFunction_MathTransformTransformVector.Result
   )
   > RigVMVariableNode(Value=@n10.Result.Rotation) {variable=TargetAimQuat}
   For_Each = RigVMDispatchNode(ExecuteContext=(), Array=RigVMVariableNode() {variable=Deform Robe Chain}.Value) {template="DISPATCH_RigVMDispatch_ArrayIterator(in Array,out Element,out Index,out Count,out Ratio)"}
   n11 = RigVMVariableNode() {variable=Deform Calf}
   RigVMFunction_MathQuaternionToEuler = &RigVMFunction_MathQuaternionToEuler(
    RotationOrder=ZYX
    Value=&RigUnit_GetRelativeTransformForItem(bChildInitial=False, bParentInitial=true, Child=@n11.Value, Parent=@n11.Value).RelativeTransform.Rotation
   )
   Remap = &RigVMFunction_MathDoubleRemap(SourceMaximum=110, SourceMinimum=0, TargetMaximum=1, TargetMinimum=0, Value=@RigVMFunction_MathQuaternionToEuler.Result.Z)
   n12 = &RigVMFunction_MathDoubleMul(A=@Remap.Result, B=RigVMVariableNode() {variable=Push Factor}.Value)
   Add = /Script/RigVM.RigVMFunction_MathVectorAdd(
    B=&RigVMFunction_MathVectorMul(A=(X=1,Y=1,Z=1), A.X=@n12.Result, A.Y=@n12.Result, A.Z=@n12.Result, B=RigVMVariableNode() {variable=Push Axis}.Value) {template="Multiply::Execute(in A,in B,out Result)"}.Result
    A=&RigUnit_GetTransform(Space=LocalSpace, bInitial=true, Item=@For_Each.Element).Transform.Translation
   ) {is=struct RigVMUnitNode, template="Add::Execute(in A,in B,out Result)"}
   RigVMFunction_AnimEvalRichCurve = &RigVMFunction_AnimEvalRichCurve(SourceMinimum=0, TargetMinimum=0, Value=@For_Each.Ratio, Curve=RigVMVariableNode() {variable=Rotation Curve}.Value)
   > RigUnit_SetTranslation = /Script/ControlRig.RigUnit_SetTranslation(
    Space=LocalSpace
    bInitial=False
    bPropagateToChildren=True
    Value=@Add.Result
    Item=@For_Each.Element
    Weight=@RigVMFunction_AnimEvalRichCurve.Result
   ) {is=struct RigVMUnitNode, template="Set Transform::Execute(in Item,in Space,in bInitial,in Value,in Weight,in bPropagateToChildren)"}
   Interpolate = /Script/RigVM.RigVMFunction_MathQuaternionSlerp(
    A=(X=0,Y=0,Z=0,W=1)
    B=RigVMDispatchNode(Array=RigVMVariableNode() {variable=TargetQuatArray}.Value, Index=@For_Each.Index) {template="DISPATCH_RigVMDispatch_ArrayGetAtIndex(in Array,in Index,out Element)"}.Element
    T=@RigVMFunction_AnimEvalRichCurve.Result
   ) {is=struct RigVMUnitNode, template="Interpolate::Execute(in A,in B,in T,out Result)"}
   > &RigUnit_SetRotation(Weight=1, bInitial=False, bPropagateToChildren=true, Item=@For_Each.Element, Value=@Interpolate.Result)
   For_Each_1 = RigVMDispatchNode(Array=(), ExecuteContext=()) {template="DISPATCH_RigVMDispatch_ArrayIterator(in Array,out Element,out Index,out Count,out Ratio)"}
   n13 = &RigVMFunction_MathDoubleMul(
    B=&RigVMFunction_MathDoubleMul(A=1, B=RigVMVariableNode() {variable=Slide Factor}.Value).Result
    A=&RigVMFunction_AnimEvalRichCurve(SourceMinimum=0, TargetMinimum=0, Value=@For_Each_1.Ratio, Curve=RigVMVariableNode() {variable=Slide Curve}.Value).Result
   )
   > &RigUnit_ModifyTransforms(
    ItemToModify=((Item=(Type=Bone,Name="None"),Transform=(Rotation=(X=0,Y=0,Z=0,W=1),Translation=(X=-5,Y=0,Z=0),Scale3D=(X=1,Y=1,Z=1))))
    Weight=1
    WeightMinimum=0
    ItemToModify.0.Item=@For_Each_1.Element
    ItemToModify.0.Transform.Translation=&RigVMFunction_MathVectorMul(B=(X=1,Y=1,Z=1), B.X=@n13.Result, B.Y=@n13.Result, B.Z=@n13.Result, A=RigVMVariableNode() {variable=Aim Primary Axis}.Value) {template="Multiply::Execute(in A,in B,out Result)"}.Result
   )
   RigVMDispatchNode(Index=-1, Array=RigVMVariableNode() {variable=TargetQuatArray}.Value) {template="DISPATCH_RigVMDispatch_ArrayGetAtIndex(in Array,in Index,out Element)"}
   RigVMRerouteNode(Value=@RigVMFunction_Sequence_1.C)
   &RigUnit_GetTransform(Item=(Type=Bone,Name="calf_l"), bInitial=False)
   /Script/RigVM.RigVMFunction_DebugTransformMutableNoSpace(
    Color=(R=1,G=0,B=0,A=1)
    DebugDrawSettings=(DepthPriority=SDPG_Foreground,Lifetime=-1)
    ExecutePin=()
    Mode=Box
    Scale=10
    Thickness=0
    Transform=(Rotation=(X=0,Y=0,Z=0,W=1),Translation=(X=0,Y=0,Z=0),Scale3D=(X=1,Y=1,Z=1))
    WorldOffset=(Rotation=(X=0,Y=0,Z=0,W=1),Translation=(X=0,Y=0,Z=0),Scale3D=(X=1,Y=1,Z=1))
    bEnabled=True
   ) {is=struct RigVMUnitNode}
   &RigVMFunction_DebugTransformMutableNoSpace(
    Color=(R=1,G=0,B=0,A=1)
    DebugDrawSettings=(DepthPriority=SDPG_Foreground,Lifetime=-1)
    ExecutePin=()
    Mode=Axes
    Scale=10
    Thickness=0
    Transform=(Rotation=(X=0,Y=0,Z=0,W=1),Translation=(X=0,Y=0,Z=0),Scale3D=(X=1,Y=1,Z=1))
    WorldOffset=(Rotation=(X=0,Y=0,Z=0,W=1),Translation=(X=0,Y=0,Z=0),Scale3D=(X=1,Y=1,Z=1))
    bEnabled=True
   )
   RigVMVariableNode() {variable=TargetAimQuat}
  ]}
 RigVMFunctionLibrary {nodes [
   Rear Robe Aim = RigVMCollapseNode(
    Aim Primary Axis=(X=-1,Y=0,Z=0)
    Aim Secondary Axis=(X=0,Y=0,Z=1)
    Bind Foot=(Type=Bone,Name="foot_l")
    Deform Calf=(Type=Bone,Name="deform_thigh_l")
    Deform Robe Chain=()
    ExecuteContext=()
    Opposite Bind Calf=(Type=Bone,Name="calf_r")
    Push Axis=(X=0,Y=0,Z=0)
    Rotation Curve=(EditorCurveData=(Keys=((Time=-0.0099,Value=0.7424),(Time=0.4972,Value=0.3128),(Time=1.0068,Value=0.0963))))
    Slide Curve=(EditorCurveData=(Keys=((Time=-0.0099,Value=0.7424),(Time=0.4972,Value=0.3128),(Time=1.0068,Value=0.0963))))
   )
   Robe Deformation = RigVMCollapseNode(ExecuteContext=())
  ]}
 RigVMFunction_Sequence_1_ContainedGraph {nodes [
   n1 = RigVMFunctionEntryNode()
   RigVMFunction_Sequence = &RigVMFunction_Sequence(ExecuteContext=@n1.ExecuteContext)
   n2 = &RigVMFunction_Sequence(ExecuteContext=@RigVMFunction_Sequence.B)
   n3 = &RigVMFunction_Sequence(ExecuteContext=@n2.B)
   n4 = &RigVMFunction_Sequence(ExecuteContext=@n3.B)
   RigVMFunctionReturnNode(A=@RigVMFunction_Sequence.A, B=@n2.A, C=@n3.A, D=@n4.A, E=@n4.B)
  ]}
 RigVMFunction_Sequence_2_ContainedGraph {nodes [
   n1 = RigVMFunctionEntryNode()
   n2 = &RigVMFunction_Sequence(ExecuteContext=@n1.ExecuteContext)
   n3 = &RigVMFunction_Sequence(ExecuteContext=@n2.B)
   n4 = &RigVMFunction_Sequence(ExecuteContext=@n3.B)
   RigVMFunctionReturnNode(A=@n2.A, B=@n3.A, C=@n4.A, D=@n4.B)
  ]}
 RigVMFunction_Sequence_2_ContainedGraph {nodes [
   n1 = RigVMFunctionEntryNode()
   n2 = &RigVMFunction_Sequence(ExecuteContext=@n1.ExecuteContext)
   n3 = &RigVMFunction_Sequence(ExecuteContext=@n2.B)
   n4 = &RigVMFunction_Sequence(ExecuteContext=@n3.B)
   RigVMFunctionReturnNode(A=@n2.A, B=@n3.A, C=@n4.A, D=@n4.B)
  ]}
 RigVMFunction_Sequence_3_ContainedGraph {nodes [
   n1 = RigVMFunctionEntryNode()
   n2 = &RigVMFunction_Sequence(ExecuteContext=@n1.ExecuteContext)
   n3 = &RigVMFunction_Sequence(ExecuteContext=@n2.B)
   n4 = &RigVMFunction_Sequence(ExecuteContext=@n3.B)
   RigVMFunctionReturnNode(B=@n3.A, A=@n2.A, C=@n4.A, D=@n4.B)
  ]}
 RigVMModel {nodes [
   RigUnit_BeginExecution = &RigUnit_BeginExecution()
   Robe Deformation = RigVMFunctionReferenceNode(ExecuteContext=@RigUnit_BeginExecution.ExecutePin)
   > Modify Transforms with Mirror = RigVMFunctionReferenceNode(
    Base Modify Data="((Item=(Type=Bone,Name=\"dyn_robe_rear_l_01\"),Transform=(Rotation=(X=0,Y=0,Z=-0.3827,W=0.9239),Translation=(X=0,Y=-2,Z=0),Scale3D=(X=1,Y=1,Z=1))),(Item=(Type=Bone,Name=\"dyn_robe_rear_l_02\"),Transform=(Rotation=(X=0,Y=0,Z=0,W=1),Translation=(X=0,Y=0,Z=0),Sca ...(truncated)"
    Mirror Modify Data="((Item=(Type=Bone,Name=\"dyn_robe_rear_r_01\"),Transform=(Rotation=(X=0,Y=0,Z=0,W=1),Translation=(X=0,Y=0,Z=0),Scale3D=(X=1,Y=1,Z=1))),(Item=(Type=Bone,Name=\"dyn_robe_rear_r_02\"),Transform=(Rotation=(X=0,Y=0,Z=0,W=1),Translation=(X=0,Y=0,Z=0),Scale3D=(X=1,Y= ...(truncated)"
    Weight=@"Robe Deformation".L Knee Bend
    "Mirror Weight "=@"Robe Deformation".R Knee Bend
   )
   > Modify Transforms with Mirror_1 = RigVMFunctionReferenceNode(
    Base Modify Data="((Item=(Type=Bone,Name=\"deform_thigh_l\"),Transform=(Rotation=(X=0,Y=0,Z=0,W=1),Translation=(X=0,Y=0,Z=0),Scale3D=(X=1,Y=1,Z=1))),(Item=(Type=Bone,Name=\"dyn_robe_rear_l_01\"),Transform=(Rotation=(X=0,Y=0.0872,Z=0,W=0.9962),Translation=(X=0,Y=0,Z=0),Scale3D=( ...(truncated)"
    Mirror Modify Data="((Item=(Type=Bone,Name=\"deform_thigh_r\"),Transform=(Rotation=(X=0,Y=0,Z=0,W=1),Translation=(X=0,Y=0,Z=0),Scale3D=(X=1,Y=1,Z=1))),(Item=(Type=Bone,Name=\"dyn_robe_rear_r_01\"),Transform=(Rotation=(X=0,Y=0,Z=0,W=1),Translation=(X=0,Y=0,Z=0),Scale3D=(X=1,Y=1,Z= ...(truncated)"
    Weight=@"Robe Deformation".L Foot Across Rear
    "Mirror Weight "=@"Robe Deformation".R Foot Across Rear
   )
   > Modify Transforms with Mirror_2 = RigVMFunctionReferenceNode(
    Base Modify Data="((Item=(Type=Bone,Name=\"deform_thigh_l\"),Transform=(Rotation=(X=0,Y=0,Z=0,W=1),Translation=(X=0,Y=0,Z=0),Scale3D=(X=1,Y=1,Z=1))),(Item=(Type=Bone,Name=\"dyn_robe_front_r_01\"),Transform=(Rotation=(X=0,Y=0,Z=0,W=1),Translation=(X=0,Y=0,Z=0),Scale3D=(X=1,Y=1,Z ...(truncated)"
    Mirror Modify Data="((Item=(Type=Bone,Name=\"deform_thigh_r\"),Transform=(Rotation=(X=0,Y=0,Z=0,W=1),Translation=(X=0,Y=0,Z=0),Scale3D=(X=1,Y=1,Z=1))),(Item=(Type=Bone,Name=\"dyn_robe_front_l_01\"),Transform=(Rotation=(X=0,Y=0,Z=0,W=1),Translation=(X=0,Y=0,Z=0),Scale3D=(X=1,Y=1,Z ...(truncated)"
    Weight=@"Robe Deformation".L Foot Across Front
    "Mirror Weight "=@"Robe Deformation".R Foot Across Front
   )
  ]}
 RigVMModel Rig {nodes [&RigUnit_PrepareForExecution()]}
]
hierarchy [
 root {loc=<0 0 0>, rot=<0 0 0 1>}
 pelvis {parent=root, loc=<0 -0.7071 93.1191>, rot=<0.0127 0.707 -0.0127 -0.707>}
 spine_01 {parent=pelvis, loc=<6.7494 -0.6388 0>, rot=<0 0 0.034 -0.9994>}
 spine_02 {parent=spine_01, loc=<7.2518 -0.035 0>, rot=<0 0 0.0238 -0.9997>}
 spine_03 {parent=spine_02, loc=<6.0104 0 0>, rot=<0 0 0.0149 -0.9999>}
 spine_04 {parent=spine_03, loc=<8.1024 0 0>, rot=<0 0 -0.0132 -0.9999>}
 spine_05 {parent=spine_04, loc=<10.0146 0 0>, rot=<0 0 -0.0015 -1>}
 neck_01 {parent=spine_05, loc=<11.3138 0.1529 0>, rot=<0 0 -0.2003 -0.9797>}
 neck_02 {parent=neck_01, loc=<4.5627 0 0>, rot=<0 0 0.0504 -0.9987>}
 head {parent=neck_02, loc=<5.3914 0 0>, rot=<0 0 0.115 -0.9934>}
 attach_cape {parent=spine_05, loc=<8.9532 -11.681 0.0524>, rot=<0.0616 -0.7044 -0.0616 -0.7044>}
 cape {parent=attach_cape, type=Null, loc=<0 0 0>, rot=<0 0 0 1>}
 attach_backpack {parent=spine_05, loc=<-0.3552 -13.6178 0.0435>, rot=<-0.0093 -0.707 0.0092 -0.707>}
 Backpack_BR {parent=attach_backpack, type=Null, loc=<0 -0.5774 0.0898>, rot=<-0.0771 0 0 0.997>}
 clavicle_l {parent=spine_05, loc=<4.3575 -0.1466 -2.3223>, rot=<0.0465 -0.7289 -0.009 -0.6829>}
 upperarm_l {parent=clavicle_l, loc=<15.5111 0 0>, rot=<0.0226 -0.3234 -0.0088 -0.9459>}
 lowerarm_l {parent=upperarm_l, loc=<26.6514 0 0>, rot=<0.0391 -0.0099 -0.2453 -0.9686>}
 hand_l {parent=lowerarm_l, loc=<22.3491 0 0>, rot=<-0.7081 -0.0507 0.0975 -0.6975>}
 thumb_01_l {parent=hand_l, loc=<1.3118 -0.3191 -2.459>, rot=<0.6546 -0.4042 0.0664 -0.6354>}
 thumb_02_l {parent=thumb_01_l, loc=<5.9336 -0.0448 0.0401>, rot=<-0.0037 0.0213 0.298 -0.9543>}
 thumb_03_l {parent=thumb_02_l, loc=<2.2223 0.3682 -0.0907>, rot=<0.0118 -0.013 0.0066 -0.9998>}
 index_metacarpal_l {parent=hand_l, loc=<1.6535 0.0183 -2.96>, rot=<0.0158 -0.0124 -0.0173 -0.9996>}
 index_01_l {parent=index_metacarpal_l, loc=<8.3008 0.0357 -0.0524>, rot=<-0.0044 0.0485 0.0924 -0.9945>}
 index_02_l {parent=index_01_l, loc=<3.3587 -0.3004 0.0452>, rot=<-0.002 0.0027 0.3225 -0.9466>}
 index_03_l {parent=index_02_l, loc=<2.5578 -0.123 -0.0252>, rot=<-0.002 0.0067 0.2558 -0.9667>}
 middle_metacarpal_l {parent=hand_l, loc=<2.1856 -0.1725 -0.5984>, rot=<0.006 0.0168 -0.0329 -0.9993>}
 middle_01_l {parent=middle_metacarpal_l, loc=<7.3275 -0.2447 0.0129>, rot=<-0.0208 0.0042 0.0858 -0.9961>}
 middle_02_l {parent=middle_01_l, loc=<4.0123 0.1401 0.1175>, rot=<-0.0218 0.0258 0.3512 -0.9357>}
 middle_03_l {parent=middle_02_l, loc=<2.6077 0.0534 0.1373>, rot=<0.0082 0.0188 0.1708 -0.9851>}
 ring_metacarpal_l {parent=hand_l, loc=<2.2074 -0.3625 1.3638>, rot=<-0.0284 0.0722 -0.0458 -0.9959>}
 ring_01_l {parent=ring_metacarpal_l, loc=<7.0295 -0.6197 -0.1446>, rot=<-0.0489 -0.0252 0.1892 -0.9804>}
 ring_02_l {parent=ring_01_l, loc=<3.6174 0.4647 -0.1027>, rot=<0 0 0.2806 -0.9598>}
 ring_03_l {parent=ring_02_l, loc=<2.4152 0.114 0.0199>, rot=<0 0 0.1189 -0.9929>}
 pinky_metacarpal_l {parent=hand_l, loc=<1.7616 -0.9753 3.3814>, rot=<-0.0397 0.0814 -0.069 -0.9935>}
 pinky_01_l {parent=pinky_metacarpal_l, loc=<7.1709 -0.4831 -0.1094>, rot=<-0.0654 -0.0035 0.2242 -0.9723>}
 pinky_02_l {parent=pinky_01_l, loc=<2.8274 0.0893 -0.1043>, rot=<0.0037 0.0011 0.2854 -0.9584>}
 pinky_03_l {parent=pinky_02_l, loc=<2.2046 0.0307 -0.0078>, rot=<-0.0002 0.0098 0.1445 -0.9895>}
 deform_wrist_dn_base_l {parent=hand_l, loc=<0 -2.2342 0>, rot=<0 0 0 -1>}
 deform_wrist_dn_l {parent=deform_wrist_dn_base_l, loc=<0 0 0>, rot=<0 0 0 -1>}
 deform_wrist_up_base_l {parent=hand_l, loc=<0 2.2682 0>, rot=<0 0 0 -1>}
 deform_wrist_up_l {parent=deform_wrist_up_base_l, loc=<0 0 0>, rot=<0 0 0 -1>}
 lowerarm_twist_02_l {parent=lowerarm_l, loc=<9.6199 0 0>, rot=<0 0 0 -1>}
 lowerarm_twist_01_l {parent=lowerarm_l, loc=<17.8549 0 0>, rot=<0 0 0 -1>}
 deform_elbow_in_base_l {parent=lowerarm_l, loc=<0.2112 5.259 0.8274>, rot=<0 0 0 -1>}
 deform_elbow_in_l {parent=deform_elbow_in_base_l, loc=<0 0 0>, rot=<0 0 0 1>}
 deform_elbow_out_base_l {parent=lowerarm_l, loc=<0 -4 0>, rot=<0 0 0 -1>}
 deform_elbow_out_l {parent=deform_elbow_out_base_l, loc=<0 0 0>, rot=<0 0 0 1>}
 upperarm_twist_01_l {parent=upperarm_l, loc=<0.0041 0 0>, rot=<0 0 0 -1>}
 deform_upperarm_bk_base_l {parent=upperarm_twist_01_l, loc=<1.0813 -7.2193 -1.9453>, rot=<0 0 0 1>}
 deform_upperarm_bk_l {parent=deform_upperarm_bk_base_l, loc=<0 0 0>, rot=<0 0 0 1>}
 deform_upperarm_fr_base_l {parent=upperarm_twist_01_l, loc=<-0.3691 7.0506 -2.765>, rot=<0 0 0 1>}
 deform_upperarm_fr_l {parent=deform_upperarm_fr_base_l, loc=<0 0 0>, rot=<0 0 0 1>}
 deform_upperarm_up_base_l {parent=upperarm_twist_01_l, loc=<-3.2216 -1.1391 3.3311>, rot=<0 0 0 1>}
 deform_upperarm_up_l {parent=deform_upperarm_up_base_l, loc=<0 0 0>, rot=<0 0 0 1>}
 upperarm_twist_02_l {parent=upperarm_l, loc=<13.7941 0 0>, rot=<0 0 0 -1>}
 dyn_upperarm_sleeve_l_MSTR {parent=upperarm_l, loc=<-0.0001 0 -0.0001>, rot=<0 0 0 -1>}
 dyn_upperarm_l_sleeve_twist_01_MSTR {parent=dyn_upperarm_sleeve_l_MSTR, loc=<10.0366 0 -0.1>, rot=<0 0 0 1>}
 dyn_upperarm_l_sleeve_twist_01 {parent=dyn_upperarm_l_sleeve_twist_01_MSTR, loc=<0 0 0>, rot=<0 0 0 1>}
 dyn_upperarm_l_sleeve_twist_01_end {parent=dyn_upperarm_l_sleeve_twist_01, loc=<0 0 -6>, rot=<0 0 0 1>}
 dyn_upperarm_l_sleeve_twist_02_MSTR {parent=dyn_upperarm_sleeve_l_MSTR, loc=<15.3484 0 -0.1>, rot=<0 0 0 -1>}
 dyn_upperarm_l_sleeve_twist_02 {parent=dyn_upperarm_l_sleeve_twist_02_MSTR, loc=<0 0 0>, rot=<0 0 0 -1>}
 dyn_upperarm_l_sleeve_twist_02_end {parent=dyn_upperarm_l_sleeve_twist_02, loc=<0 0 -7>, rot=<0 0 0 1>}
 dyn_upperarm_l_sleeve_twist_03_MSTR {parent=dyn_upperarm_sleeve_l_MSTR, loc=<20.9241 0 -0.1>, rot=<0 0 0 -1>}
 dyn_upperarm_l_sleeve_twist_03 {parent=dyn_upperarm_l_sleeve_twist_03_MSTR, loc=<0 0 0>, rot=<0 0 0 -1>}
 dyn_upperarm_l_sleeve_twist_03_end {parent=dyn_upperarm_l_sleeve_twist_03, loc=<0 0 -8>, rot=<0 0 0 1>}
 clavicle_r {parent=spine_05, loc=<4.3575 -0.1466 2.3223>, rot=<0.7289 0.0465 -0.6829 0.009>}
 upperarm_r {parent=clavicle_r, loc=<-15.5111 0 0>, rot=<0.0226 -0.3234 -0.0088 -0.9459>}
 lowerarm_r {parent=upperarm_r, loc=<-26.6514 0 0>, rot=<0.0391 -0.0099 -0.2453 -0.9686>}
 hand_r {parent=lowerarm_r, loc=<-22.3491 0 0>, rot=<-0.7081 -0.0507 0.0975 -0.6975>}
 thumb_01_r {parent=hand_r, loc=<-1.3118 0.3191 2.459>, rot=<0.6546 -0.4042 0.0664 -0.6354>}
 thumb_02_r {parent=thumb_01_r, loc=<-5.9336 0.0448 -0.0401>, rot=<-0.0037 0.0213 0.298 -0.9543>}
 thumb_03_r {parent=thumb_02_r, loc=<-2.2223 -0.3682 0.0907>, rot=<0.0118 -0.013 0.0066 -0.9998>}
 index_metacarpal_r {parent=hand_r, loc=<-1.6535 -0.0183 2.96>, rot=<0.0158 -0.0124 -0.0173 -0.9996>}
 index_01_r {parent=index_metacarpal_r, loc=<-8.3008 -0.0357 0.0524>, rot=<-0.0044 0.0485 0.0924 -0.9945>}
 index_02_r {parent=index_01_r, loc=<-3.3587 0.3004 -0.0452>, rot=<-0.002 0.0027 0.3225 -0.9466>}
 index_03_r {parent=index_02_r, loc=<-2.5578 0.123 0.0252>, rot=<-0.002 0.0067 0.2558 -0.9667>}
 middle_metacarpal_r {parent=hand_r, loc=<-2.1856 0.1725 0.5984>, rot=<0.006 0.0168 -0.0329 -0.9993>}
 middle_01_r {parent=middle_metacarpal_r, loc=<-7.3275 0.2447 -0.0129>, rot=<-0.0208 0.0042 0.0858 -0.9961>}
 middle_02_r {parent=middle_01_r, loc=<-4.0123 -0.1401 -0.1175>, rot=<-0.0218 0.0258 0.3512 -0.9357>}
 middle_03_r {parent=middle_02_r, loc=<-2.6077 -0.0534 -0.1373>, rot=<0.0082 0.0188 0.1708 -0.9851>}
 ring_metacarpal_r {parent=hand_r, loc=<-2.2074 0.3625 -1.3638>, rot=<-0.0284 0.0722 -0.0458 -0.9959>}
 ring_01_r {parent=ring_metacarpal_r, loc=<-7.0295 0.6197 0.1446>, rot=<-0.0489 -0.0252 0.1892 -0.9804>}
 ring_02_r {parent=ring_01_r, loc=<-3.6174 -0.4647 0.1027>, rot=<0 0 0.2806 -0.9598>}
 ring_03_r {parent=ring_02_r, loc=<-2.4152 -0.114 -0.0199>, rot=<0 0 0.1189 -0.9929>}
 pinky_metacarpal_r {parent=hand_r, loc=<-1.7616 0.9753 -3.3814>, rot=<-0.0397 0.0814 -0.069 -0.9935>}
 pinky_01_r {parent=pinky_metacarpal_r, loc=<-7.1709 0.4831 0.1094>, rot=<-0.0654 -0.0035 0.2242 -0.9723>}
 pinky_02_r {parent=pinky_01_r, loc=<-2.8274 -0.0893 0.1043>, rot=<0.0037 0.0011 0.2854 -0.9584>}
 pinky_03_r {parent=pinky_02_r, loc=<-2.2046 -0.0307 0.0078>, rot=<-0.0002 0.0098 0.1445 -0.9895>}
 deform_wrist_dn_base_r {parent=hand_r, loc=<-0.0004 2.2345 -0.0001>, rot=<0 0 0 -1>}
 deform_wrist_dn_r {parent=deform_wrist_dn_base_r, loc=<0 0 0>, rot=<0 0 0 -1>}
 deform_wrist_up_base_r {parent=hand_r, loc=<0.0003 -2.2684 0.0001>, rot=<0 0 0 -1>}
 deform_wrist_up_r {parent=deform_wrist_up_base_r, loc=<0 0 0>, rot=<0 0 0 -1>}
 lowerarm_twist_02_r {parent=lowerarm_r, loc=<-9.6199 0 0>, rot=<0 0 0 -1>}
 lowerarm_twist_01_r {parent=lowerarm_r, loc=<-17.8549 0 0>, rot=<0 0 0 -1>}
 deform_elbow_in_base_r {parent=lowerarm_r, loc=<-0.2111 -5.2591 -0.8275>, rot=<0 0 0 -1>}
 deform_elbow_in_r {parent=deform_elbow_in_base_r, loc=<-0.0001 0 0.0002>, rot=<0 0 0 1>}
 deform_elbow_out_base_r {parent=lowerarm_r, loc=<0 4 -0.0001>, rot=<0 0 0 -1>}
 deform_elbow_out_r {parent=deform_elbow_out_base_r, loc=<0 0 0>, rot=<0 0 0 1>}
 upperarm_twist_01_r {parent=upperarm_r, loc=<-0.0041 0 0>, rot=<0 0 0 -1>}
 deform_upperarm_bk_base_r {parent=upperarm_twist_01_r, loc=<-1.0814 7.2192 1.9455>, rot=<0 0 0 1>}
 deform_upperarm_bk_r {parent=deform_upperarm_bk_base_r, loc=<0 0 0>, rot=<0 0 0 1>}
 deform_upperarm_fr_base_r {parent=upperarm_twist_01_r, loc=<0.3689 -7.0506 2.7652>, rot=<0 0 0 1>}
 deform_upperarm_fr_r {parent=deform_upperarm_fr_base_r, loc=<0 0 0>, rot=<0 0 0 1>}
 deform_upperarm_up_base_r {parent=upperarm_twist_01_r, loc=<3.2214 1.1391 -3.3308>, rot=<0 0 0 1>}
 deform_upperarm_up_r {parent=deform_upperarm_up_base_r, loc=<0 0 0>, rot=<0 0 0 -1>}
 upperarm_twist_02_r {parent=upperarm_r, loc=<-13.7941 0 0>, rot=<0 0 0 -1>}
 dyn_upperarm_sleeve_r_MSTR {parent=upperarm_r, loc=<0 0 0.0002>, rot=<0 0 0 -1>}
 dyn_upperarm_r_sleeve_twist_01_MSTR {parent=dyn_upperarm_sleeve_r_MSTR, loc=<-10.0368 0 0.1002>, rot=<0 0 0 -1>}
 dyn_upperarm_r_sleeve_twist_01 {parent=dyn_upperarm_r_sleeve_twist_01_MSTR, loc=<0 0 0>, rot=<0 0 0 -1>}
 dyn_upperarm_r_sleeve_twist_01_end {parent=dyn_upperarm_r_sleeve_twist_01, loc=<0.0004 0 5.9995>, rot=<0 0 0 -1>}
 dyn_upperarm_r_sleeve_twist_02_MSTR {parent=dyn_upperarm_sleeve_r_MSTR, loc=<-15.348 0 0.0995>, rot=<0 0 0 -1>}
 dyn_upperarm_r_sleeve_twist_02 {parent=dyn_upperarm_r_sleeve_twist_02_MSTR, loc=<0 0 0>, rot=<0 0 0 -1>}
 dyn_upperarm_r_sleeve_twist_02_end {parent=dyn_upperarm_r_sleeve_twist_02, loc=<-0.0003 0 7.0003>, rot=<0 0 0 -1>}
 dyn_upperarm_r_sleeve_twist_03_MSTR {parent=dyn_upperarm_sleeve_r_MSTR, loc=<-20.9243 0 0.1002>, rot=<0 0 0 -1>}
 dyn_upperarm_r_sleeve_twist_03 {parent=dyn_upperarm_r_sleeve_twist_03_MSTR, loc=<0 0 0>, rot=<0 0 0 -1>}
 dyn_upperarm_r_sleeve_twist_03_end {parent=dyn_upperarm_r_sleeve_twist_03, loc=<0.0004 0 7.9996>, rot=<0 0 0 -1>}
 deform_brooch {parent=clavicle_r, loc=<-12.811 -8.7106 -3.8223>, rot=<-0.5826 0.2566 0.7124 -0.2953>}
 deform_pec_aim_base_l {parent=spine_04, loc=<7.8732 12.6388 -2.4123>, rot=<0.0102 -0.4602 0.3979 -0.7936>}
 deform_pec_aim_l {parent=deform_pec_aim_base_l, loc=<0 0 0>, rot=<0 0 0 1>}
 deform_pec_base_l {parent=deform_pec_aim_l, loc=<10.766 0 0>, rot=<0 0 0 1>}
 deform_pec_l {parent=deform_pec_base_l, loc=<0.055 0 0>, rot=<0 0 0 -1>}
 deform_pec_aim_base_r {parent=spine_04, loc=<7.8727 12.6388 2.4123>, rot=<-0.4602 -0.0102 0.7936 0.3979>}
 deform_pec_aim_r {parent=deform_pec_aim_base_r, loc=<0 0 0>, rot=<0 0 0 1>}
 deform_pec_base_r {parent=deform_pec_aim_r, loc=<-10.7662 -0.0006 -0.0007>, rot=<0 0 0 1>}
 deform_pec_r {parent=deform_pec_base_r, loc=<0 0 0>, rot=<0 0 0 -1>}
 dyn_fold_front1_MSTR {parent=spine_04, loc=<7.4931 12.5839 12.9878>, rot=<-0.1537 -0.1614 0.9711 0.0854>}
 dyn_fold_front1_01 {parent=dyn_fold_front1_MSTR, loc=<0 0 0>, rot=<0 0 0 1>}
 dyn_fold_front1_02 {parent=dyn_fold_front1_01, loc=<7.723 0 0>, rot=<0 0 0 1>}
 dyn_fold_front1_03 {parent=dyn_fold_front1_02, loc=<6.029 0 0>, rot=<0 0 0 1>}
 dyn_fold_front2_MSTR {parent=spine_04, loc=<4.7076 10.8629 17.0542>, rot=<0.0762 -0.2261 0.9711 -0.0068>}
 dyn_fold_front2_01 {parent=dyn_fold_front2_MSTR, loc=<0 0 0>, rot=<0 0 0 1>}
 dyn_fold_front2_02 {parent=dyn_fold_front2_01, loc=<6.3042 0 0>, rot=<0 0 0 1>}
 dyn_fold_front2_03 {parent=dyn_fold_front2_02, loc=<6.5675 0 0>, rot=<0 0 0 1>}
 dyn_skirt_MSTR {parent=spine_01, loc=<0 0 0>, rot=<0 0 0 1>}
 dyn_skirt_front_mid_MSTR {parent=dyn_skirt_MSTR, loc=<-1.4158 14.6109 0>, rot=<0 0 1 0.0093>}
 dyn_skirt_side_l_MSTR {parent=dyn_skirt_MSTR, loc=<-1.2346 -2.4178 -16.4314>, rot=<0.0575 -0.7145 -0.6964 -0.0335>}
 dyn_skirt_back_l_MSTR {parent=dyn_skirt_MSTR, loc=<-0.7173 -11.3926 -10.69>, rot=<0.0556 -0.9978 -0.002 -0.0362>}
 dyn_skirt_back_mid_MSTR {parent=dyn_skirt_MSTR, loc=<-0.7364 -14.5993 0>, rot=<0.0649 -0.9979 0 0>}
 dyn_skirt_front_l_MSTR {parent=dyn_skirt_MSTR, loc=<-1.2526 10.3527 -12.013>, rot=<-0.0218 0.1163 0.993 0.0041>}
 dyn_skirt_front_r_MSTR {parent=dyn_skirt_MSTR, loc=<-1.2527 10.3527 12.013>, rot=<-0.1163 -0.0218 0.0041 -0.993>}
 dyn_skirt_side_r_MSTR {parent=dyn_skirt_MSTR, loc=<-1.2346 -2.4178 16.4314>, rot=<-0.7145 -0.0575 0.0335 -0.6964>}
 dyn_skirt_back_r_MSTR {parent=dyn_skirt_MSTR, loc=<-0.7172 -11.3925 10.69>, rot=<0.9978 0.0556 -0.0362 0.002>}
 dyn_skirt_back_r_base {parent=dyn_skirt_back_r_MSTR, loc=<-27.2315 0 0>, rot=<0 0 0 1>}
 dyn_skirt_SIMSPACE {parent=spine_01, loc=<0 0 0>, rot=<0 0 0 1>}
 thigh_l {parent=pelvis, loc=<-2.5693 -0.0109 -7.7501>, rot=<0.0012 0.0572 0.0155 -0.9982>}
 calf_l {parent=thigh_l, loc=<-39.0484 0 0>, rot=<0.0006 -0.0161 -0.0391 -0.9991>}
 foot_l {parent=calf_l, loc=<-42.561 0.1077 0.4292>, rot=<0.0418 -0.0455 0.0343 -0.9975>}
 ball_l {parent=foot_l, loc=<-6.1599 13.8264 0.6806>, rot=<0.0001 -0.0001 -0.7063 -0.7079>}
 calf_twist_02_l {parent=calf_l, loc=<-18.8064 -0.0967 0.0397>, rot=<0.0237 0.0095 -0.0035 -0.9997>}
 calf_twist_01_l {parent=calf_l, loc=<-35.813 -0.0967 0.0397>, rot=<0.0237 0.0095 -0.0035 -0.9997>}
 deform_knee_in_base_l {parent=calf_l, loc=<1.298 -5.2957 -0.0496>, rot=<0 0 0 -1>}
 deform_knee_in_l {parent=deform_knee_in_base_l, loc=<0 0 0>, rot=<0 0 0 1>}
 deform_knee_out_base_l {parent=calf_l, loc=<0 4 0>, rot=<0 0 0 -1>}
 deform_knee_out_l {parent=deform_knee_out_base_l, loc=<0 0 0>, rot=<0 0 0 1>}
 thigh_twist_01_l {parent=thigh_l, loc=<-21.6303 0 0>, rot=<0.0013 0 0 -1>}
 deform_groin_base_l {parent=thigh_l, loc=<-2.1908 8.1142 -1.7442>, rot=<0 0 0 -1>}
 deform_groin_l {parent=deform_groin_base_l, loc=<0 0 0>, rot=<0 0 0 -1>}
 thigh_r {parent=pelvis, loc=<-2.5693 -0.0109 7.7501>, rot=<0.0572 -0.0012 0.9982 0.0155>}
 calf_r {parent=thigh_r, loc=<39.0484 0 0>, rot=<0.0006 -0.0161 -0.0391 -0.9991>}
 foot_r {parent=calf_r, loc=<42.561 -0.1077 -0.4292>, rot=<0.0418 -0.0455 0.0343 -0.9975>}
 ball_r {parent=foot_r, loc=<6.1599 -13.8264 -0.6806>, rot=<0.0001 -0.0001 -0.7063 -0.7079>}
 calf_twist_02_r {parent=calf_r, loc=<18.8064 0.0967 -0.0397>, rot=<0.0237 0.0095 -0.0035 -0.9997>}
 calf_twist_01_r {parent=calf_r, loc=<35.813 0.0967 -0.0397>, rot=<0.0237 0.0095 -0.0035 -0.9997>}
 deform_knee_in_base_r {parent=calf_r, loc=<-1.298 5.2957 0.0496>, rot=<0 0 0 -1>}
 deform_knee_in_r {parent=deform_knee_in_base_r, loc=<0 0 0>, rot=<0 0 0 1>}
 deform_knee_out_base_r {parent=calf_r, loc=<0 -4 0>, rot=<0 0 0 -1>}
 deform_knee_out_r {parent=deform_knee_out_base_r, loc=<0 0 0>, rot=<0 0 0 1>}
 thigh_twist_01_r {parent=thigh_r, loc=<21.6303 0 0>, rot=<0.0013 0 0 -1>}
 deform_groin_base_r {parent=thigh_r, loc=<2.1908 -8.1142 1.7442>, rot=<0 0 0 -1>}
 deform_groin_r {parent=deform_groin_base_r, loc=<0 0 0>, rot=<0 0 0 -1>}
 deform_thigh_notwist_l {parent=pelvis, loc=<-2.5785 -0.0168 -7.6736>, rot=<-0.0807 0.0001 0 -0.9967>}
 deform_thigh_notwist_r {parent=pelvis, loc=<-2.5386 -0.0122 7.6605>, rot=<0.0572 -0.0012 0.9982 0.0155>}
 deform_glute_base_l {parent=pelvis, loc=<-8.8064 -8.8222 -8.4287>, rot=<0.0275 0.0501 -0.4472 -0.8926>}
 deform_glute_l {parent=deform_glute_base_l, loc=<0 0 0>, rot=<0 0 0 1>}
 deform_glute_base_r {parent=pelvis, loc=<-8.8062 -8.8222 8.4287>, rot=<0.0501 -0.0275 0.8926 -0.4472>}
 deform_glute_r {parent=deform_glute_base_r, loc=<0 0 0>, rot=<0 0 0 1>}
 deform_thigh_l {parent=pelvis, loc=<-2.5693 -0.0109 -7.7501>, rot=<0.0012 0.0572 0.0155 -0.9982>}
 deform_calf_l {parent=deform_thigh_l, loc=<-39.0484 0 0>, rot=<0.0006 -0.0161 -0.0391 -0.9991>}
 dyn_robe_front_l_base {parent=deform_thigh_l, loc=<-19.6616 9.2343 -1.2069>, rot=<0.0158 0.0692 0.9968 0.0373>}
 dyn_robe_front_l_01 {parent=dyn_robe_front_l_base, loc=<0 0 0>, rot=<0 0 0 1>}
 dyn_robe_front_l_knee_base {parent=dyn_robe_front_l_01, loc=<18.2739 0.1389 0>, rot=<0 0 0 1>}
 dyn_robe_front_l_02 {parent=dyn_robe_front_l_01, loc=<18.2739 0.1389 0>, rot=<0 0 0 1>}
 dyn_robe_front_l_03 {parent=dyn_robe_front_l_02, loc=<10.5 0 0>, rot=<0 0 0 1>}
 dyn_robe_front_l_04 {parent=dyn_robe_front_l_03, loc=<10.5 0 0>, rot=<0 0 0 1>}
 dyn_robe_front_l_05 {parent=dyn_robe_front_l_04, loc=<10.5 0 0>, rot=<0 0 0 1>}
 dyn_robe_side_l_base {parent=deform_thigh_l, loc=<-20.6184 -1.6637 -9.0549>, rot=<0.0276 -0.7339 -0.6787 -0.0052>}
 dyn_robe_side_l_01 {parent=dyn_robe_side_l_base, loc=<0 0 0>, rot=<0 0 0 1>}
 dyn_robe_side_l_knee_base {parent=dyn_robe_side_l_01, loc=<18.2739 0.1389 0>, rot=<0 0 0 1>}
 dyn_robe_side_l_02 {parent=dyn_robe_side_l_01, loc=<18.2739 0.1389 0>, rot=<0 0 0 1>}
 dyn_robe_side_l_03 {parent=dyn_robe_side_l_02, loc=<10.5 0 0>, rot=<0 0 0 1>}
 dyn_robe_side_l_04 {parent=dyn_robe_side_l_03, loc=<10.5 0 0>, rot=<0 0 0 1>}
 dyn_robe_side_l_05 {parent=dyn_robe_side_l_04, loc=<10.5 0 0>, rot=<0 0 0 1>}
 dyn_robe_rear_l_base {parent=deform_thigh_l, loc=<-19.573 -13.6488 0.023>, rot=<0.0423 -0.9986 -0.0297 0.0122>}
 dyn_robe_rear_l_01 {parent=dyn_robe_rear_l_base, loc=<0 0 0>, rot=<0 0 0 1>}
 dyn_robe_rear_l_knee_base {parent=dyn_robe_rear_l_01, loc=<18.2739 0.1389 0>, rot=<0 0 0 1>}
 dyn_robe_rear_l_02 {parent=dyn_robe_rear_l_01, loc=<18.2739 0.1389 0>, rot=<0 0 0 1>}
 dyn_robe_rear_l_03 {parent=dyn_robe_rear_l_02, loc=<10.5 0 0>, rot=<0 0 0 1>}
 dyn_robe_rear_l_04 {parent=dyn_robe_rear_l_03, loc=<10.5 0 0>, rot=<0 0 0 1>}
 dyn_robe_rear_l_05 {parent=dyn_robe_rear_l_04, loc=<10.5 0 0>, rot=<0 0 0 1>}
 deform_thigh_r {parent=pelvis, loc=<-2.5693 -0.0109 7.7501>, rot=<0.0572 -0.0012 0.9982 0.0155>}
 deform_calf_r {parent=deform_thigh_r, loc=<39.0484 0 0>, rot=<0.0006 -0.0161 -0.0391 -0.9991>}
 dyn_robe_front_r_base {parent=deform_thigh_r, loc=<19.6617 -9.2343 1.207>, rot=<0.0158 0.0692 0.9968 0.0373>}
 dyn_robe_front_r_01 {parent=dyn_robe_front_r_base, loc=<0 0 0>, rot=<0 0 0 1>}
 dyn_robe_front_r_knee_base {parent=dyn_robe_front_r_01, loc=<-18.2738 -0.1389 0>, rot=<0 0 0 1>}
 dyn_robe_front_r_02 {parent=dyn_robe_front_r_01, loc=<-18.2738 -0.1389 0>, rot=<0 0 0 1>}
 dyn_robe_front_r_03 {parent=dyn_robe_front_r_02, loc=<-10.5001 0 0>, rot=<0 0 0 1>}
 dyn_robe_front_r_04 {parent=dyn_robe_front_r_03, loc=<-10.5 0 0>, rot=<0 0 0 1>}
 dyn_robe_front_r_05 {parent=dyn_robe_front_r_04, loc=<-10.5001 0.0001 0>, rot=<0 0 0 1>}
 dyn_robe_side_r_base {parent=deform_thigh_r, loc=<20.6184 1.6637 9.0549>, rot=<0.0276 -0.7339 -0.6787 -0.0052>}
 dyn_robe_side_r_01 {parent=dyn_robe_side_r_base, loc=<0 0 0>, rot=<0 0 0 1>}
 dyn_robe_side_r_knee_base {parent=dyn_robe_side_r_01, loc=<-18.2739 -0.1389 0>, rot=<0 0 0 1>}
 dyn_robe_side_r_02 {parent=dyn_robe_side_r_01, loc=<-18.2739 -0.1389 0>, rot=<0 0 0 1>}
 dyn_robe_side_r_03 {parent=dyn_robe_side_r_02, loc=<-10.5 0 0>, rot=<0 0 0 1>}
 dyn_robe_side_r_04 {parent=dyn_robe_side_r_03, loc=<-10.5 0 0>, rot=<0 0 0 1>}
 dyn_robe_side_r_05 {parent=dyn_robe_side_r_04, loc=<-10.5001 0.0001 0>, rot=<0 0 0 1>}
 dyn_robe_rear_r_base {parent=deform_thigh_r, loc=<19.573 13.6488 -0.023>, rot=<0.0423 -0.9986 -0.0297 0.0122>}
 dyn_robe_rear_r_01 {parent=dyn_robe_rear_r_base, loc=<0 0 0>, rot=<0 0 0 1>}
 dyn_robe_rear_r_knee_base {parent=dyn_robe_rear_r_01, loc=<-18.2738 -0.1389 0>, rot=<0 0 0 1>}
 dyn_robe_rear_r_02 {parent=dyn_robe_rear_r_01, loc=<-18.2738 -0.1389 0>, rot=<0 0 0 1>}
 dyn_robe_rear_r_03 {parent=dyn_robe_rear_r_02, loc=<-10.5 0.0001 -0.0001>, rot=<0 0 0 1>}
 dyn_robe_rear_r_04 {parent=dyn_robe_rear_r_03, loc=<-10.5001 0 0>, rot=<0 0 0 1>}
 dyn_robe_rear_r_05 {parent=dyn_robe_rear_r_04, loc=<-10.5 0 0>, rot=<0 0 0 1>}
 dyn_robe_front_mid_base {parent=pelvis, loc=<-21.952 13.0152 -0.199>, rot=<-0.005 0.0655 0.9949 0.0761>}
 dyn_robe_front_mid_01 {parent=dyn_robe_front_mid_base, loc=<0 0 0>, rot=<0 0 0 1>}
 dyn_robe_front_mid_02 {parent=dyn_robe_front_mid_01, loc=<18.2739 0.1389 0>, rot=<0 0 0 1>}
 dyn_robe_front_mid_03 {parent=dyn_robe_front_mid_02, loc=<10.5 0 0>, rot=<0 0 0 1>}
 dyn_robe_front_mid_04 {parent=dyn_robe_front_mid_03, loc=<10.5 0 0>, rot=<0 0 0 1>}
 dyn_robe_front_mid_05 {parent=dyn_robe_front_mid_04, loc=<10.5 0 0>, rot=<0 0 0 1>}
 dyn_robe_rear_mid_base {parent=pelvis, loc=<-22.1452 -14.9257 0>, rot=<0.0279 -0.9996 0 0>}
 dyn_robe_rear_mid_01 {parent=dyn_robe_rear_mid_base, loc=<0 0 0>, rot=<0 0 0 1>}
 dyn_robe_rear_mid_02 {parent=dyn_robe_rear_mid_01, loc=<18.2739 0.1389 0>, rot=<0 0 0 1>}
 dyn_robe_rear_mid_03 {parent=dyn_robe_rear_mid_02, loc=<10.5 0 0>, rot=<0 0 0 1>}
 dyn_robe_rear_mid_04 {parent=dyn_robe_rear_mid_03, loc=<10.5 0 0>, rot=<0 0 0 1>}
 dyn_robe_rear_mid_05 {parent=dyn_robe_rear_mid_04, loc=<10.5 0 0>, rot=<0 0 0 1>}
 dyn_skirt_ground_collision {parent=root, loc=<0 0 -8.2344>, rot=<0 0 0 1>}
 CR_root_no_rotation {parent=root, loc=<0 0 0>, rot=<0 0 0 1>}
 CR Foot Near Pelvis L {type=Curve, loc=<0 0 0>, rot=<0 0 0 1>}
 CR Foot Near Pelvis R {type=Curve, loc=<0 0 0>, rot=<0 0 0 1>}
 CR Rear Knee Bend L {type=Curve, loc=<0 0 0>, rot=<0 0 0 1>}
 CR Rear Knee Bend R {type=Curve, loc=<0 0 0>, rot=<0 0 0 1>}
 CR Thigh L Across R {type=Curve, loc=<0 0 0>, rot=<0 0 0 1>}
 CR Thigh R Across L {type=Curve, loc=<0 0 0>, rot=<0 0 0 1>}
 CR Thigh Raised Front L {type=Curve, loc=<0 0 0>, rot=<0 0 0 1>}
 CR Thigh Raised Front R {type=Curve, loc=<0 0 0>, rot=<0 0 0 1>}
 MaterialCurveA {type=Curve, loc=<0 0 0>, rot=<0 0 0 1>}
 dyn_skirt_side_r_06_FxSocket {type=Null, loc=<-26.5063 -3.8475 20.2014>, rot=<-0.4742 0.5243 0.5446 0.4513>}
 dyn_skirt_front_r_06_FxSocket {type=Null, loc=<-15.5042 12.071 20.3167>, rot=<-0.0682 0.7161 0.0967 0.6879>}
 dyn_skirt_back_r_06_FxSocket {type=Null, loc=<-16.4684 -18.7527 19.4533>, rot=<0.6805 -0.0289 -0.7317 0.0269>}
]
