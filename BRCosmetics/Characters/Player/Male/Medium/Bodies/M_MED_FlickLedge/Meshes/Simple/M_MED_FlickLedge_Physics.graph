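# Physics asset for the FlickLedge medium male body's simple mesh, setting up dynamic robe bones around the legs.
# Six capsule chains (front, side and rear on each leg) hang from the thigh bodies, which attach to the pelvis.
# Chain roots have wider angular limits than the tighter links below them.
# Tiny knee-base bodies join each thigh and chain root through free-rotating constraints.
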PhysicsAsset /BRCosmetics/Characters/Player/Male/Medium/Bodies/M_MED_FlickLedge/Meshes/Simple/M_MED_FlickLedge_Physics
bodies {
 deform_thigh_l [{shapeType=Capsule, length=1, radius=1, shapeName=deform_thigh_l_capsule}]
 deform_thigh_r [{shapeType=Capsule, length=1, radius=1, shapeName=deform_thigh_r_capsule}]
 dyn_robe_front_l_02 [{shapeType=Capsule, length=13.0438, radius=8}]
 dyn_robe_front_l_03 [{shapeType=Capsule, length=12.3245, radius=8, shapeName=dyn_robe_front_l_03_capsule}]
 dyn_robe_front_l_04 [{shapeType=Capsule, length=14.6324, radius=8}]
 dyn_robe_front_l_05 [{shapeType=Capsule, length=12.1055, radius=8, shapeName=dyn_robe_front_l_05_capsule}]
 dyn_robe_front_l_knee_base [{shapeType=Capsule, length=0.505, radius=0.505, shapeName=dyn_robe_front_l_knee_base_capsule, rotation=<rot 0 -90 0>}]
 dyn_robe_front_r_02 [{shapeType=Capsule, length=10.7845, radius=8}]
 dyn_robe_front_r_03 [{shapeType=Capsule, length=9.6887, radius=8, shapeName=dyn_robe_front_r_03_capsule}]
 dyn_robe_front_r_04 [{shapeType=Capsule, length=14.4698, radius=8}]
 dyn_robe_front_r_05 [{shapeType=Capsule, length=14.4912, radius=8, shapeName=dyn_robe_front_r_05_capsule}]
 dyn_robe_front_r_knee_base [{shapeType=Capsule, length=0.505, radius=0.505, shapeName=dyn_robe_front_r_knee_base_capsule, rotation=<rot 0 -90 0>}]
 dyn_robe_rear_l_02 [{shapeType=Capsule, length=13.0663, radius=8}]
 dyn_robe_rear_l_03 [{shapeType=Capsule, length=11.0186, radius=8, shapeName=dyn_robe_rear_l_03_capsule}]
 dyn_robe_rear_l_04 [{shapeType=Capsule, length=13.616, radius=8}]
 dyn_robe_rear_l_05 [{shapeType=Capsule, length=11.1164, radius=8, shapeName=dyn_robe_rear_l_05_capsule}]
 dyn_robe_rear_l_knee_base [{shapeType=Capsule, length=0.505, radius=0.505, shapeName=dyn_robe_rear_l_knee_base_capsule, rotation=<rot 0 -90 0>}]
 dyn_robe_rear_r_02 [{shapeType=Capsule, length=13.224, radius=8}]
 dyn_robe_rear_r_03 [{shapeType=Capsule, length=9.8149, radius=8, shapeName=dyn_robe_rear_r_03_capsule}]
 dyn_robe_rear_r_04 [{shapeType=Capsule, length=14.7322, radius=8}]
 dyn_robe_rear_r_05 [{shapeType=Capsule, length=13.1029, radius=8, shapeName=dyn_robe_rear_r_05_capsule}]
 dyn_robe_rear_r_knee_base [{shapeType=Capsule, length=0.505, radius=0.505, shapeName=dyn_robe_rear_r_knee_base_capsule, rotation=<rot 0 -90 0>}]
 dyn_robe_side_l_02 [{shapeType=Capsule, length=11.5284, radius=8}]
 dyn_robe_side_l_03 [{shapeType=Capsule, length=9.324, radius=8, shapeName=dyn_robe_side_l_03_capsule}]
 dyn_robe_side_l_04 [{shapeType=Capsule, length=12.5793, radius=8}]
 dyn_robe_side_l_05 [{shapeType=Capsule, length=10.9448, radius=8, shapeName=dyn_robe_side_l_05_capsule}]
 dyn_robe_side_l_knee_base [{shapeType=Capsule, length=0.505, radius=0.505, shapeName=dyn_robe_side_l_knee_base_capsule, rotation=<rot 0 -90 0>}]
 dyn_robe_side_r_02 [{shapeType=Capsule, length=13.411, radius=8}]
 dyn_robe_side_r_03 [{shapeType=Capsule, length=10.6109, radius=8, shapeName=dyn_robe_side_r_03_capsule}]
 dyn_robe_side_r_04 [{shapeType=Capsule, length=13.2922, radius=8}]
 dyn_robe_side_r_05 [{shapeType=Capsule, length=11.9713, radius=8, shapeName=dyn_robe_side_r_05_capsule}]
 dyn_robe_side_r_knee_base [{shapeType=Capsule, length=0.505, radius=0.505, shapeName=dyn_robe_side_r_knee_base_capsule, rotation=<rot 0 -90 0>}]
 pelvis [{shapeType=Capsule, length=1, radius=1}]
}
constraints [
 {bone1Name=dyn_robe_front_l_02, bone2Name=deform_thigh_l, swing1LimitDegrees=45, swing2LimitDegrees=45, twistLimitDegrees=45}
 {bone1Name=dyn_robe_front_l_04, bone2Name=dyn_robe_front_l_03, swing1LimitDegrees=12, swing2LimitDegrees=12, twistLimitDegrees=12}
 {bone1Name=dyn_robe_side_l_02, bone2Name=deform_thigh_l, swing1LimitDegrees=45, swing2LimitDegrees=45, twistLimitDegrees=45}
 {bone1Name=dyn_robe_side_l_04, bone2Name=dyn_robe_side_l_03, swing1LimitDegrees=12, swing2LimitDegrees=12, twistLimitDegrees=12}
 {bone1Name=dyn_robe_rear_l_02, bone2Name=deform_thigh_l, swing1LimitDegrees=45, swing2LimitDegrees=45, twistLimitDegrees=45}
 {bone1Name=dyn_robe_rear_l_04, bone2Name=dyn_robe_rear_l_03, swing1LimitDegrees=12, swing2LimitDegrees=12, twistLimitDegrees=12}
 {bone1Name=dyn_robe_front_r_02, bone2Name=deform_thigh_r, swing1LimitDegrees=45, swing2LimitDegrees=45, twistLimitDegrees=45}
 {bone1Name=dyn_robe_front_r_04, bone2Name=dyn_robe_front_r_03, swing1LimitDegrees=12, swing2LimitDegrees=12, twistLimitDegrees=12}
 {bone1Name=dyn_robe_side_r_02, bone2Name=deform_thigh_r, swing1LimitDegrees=45, swing2LimitDegrees=45, twistLimitDegrees=45}
 {bone1Name=dyn_robe_side_r_04, bone2Name=dyn_robe_side_r_03, swing1LimitDegrees=12, swing2LimitDegrees=12, twistLimitDegrees=12}
 {bone1Name=dyn_robe_rear_r_02, bone2Name=deform_thigh_r, swing1LimitDegrees=45, swing2LimitDegrees=45, twistLimitDegrees=45}
 {bone1Name=dyn_robe_rear_r_04, bone2Name=dyn_robe_rear_r_03, swing1LimitDegrees=12, swing2LimitDegrees=12, twistLimitDegrees=12}
 {bone1Name=dyn_robe_front_l_03, bone2Name=dyn_robe_front_l_02, swing1LimitDegrees=12, swing2LimitDegrees=12, twistLimitDegrees=12}
 {bone1Name=dyn_robe_rear_l_03, bone2Name=dyn_robe_rear_l_02, swing1LimitDegrees=12, swing2LimitDegrees=12, twistLimitDegrees=12}
 {bone1Name=dyn_robe_front_l_05, bone2Name=dyn_robe_front_l_04, swing1LimitDegrees=12, swing2LimitDegrees=12, twistLimitDegrees=12}
 {bone1Name=dyn_robe_rear_l_05, bone2Name=dyn_robe_rear_l_04, swing1LimitDegrees=12, swing2LimitDegrees=12, twistLimitDegrees=12}
 {bone1Name=dyn_robe_side_l_03, bone2Name=dyn_robe_side_l_02, swing1LimitDegrees=12, swing2LimitDegrees=12, twistLimitDegrees=12}
 {bone1Name=dyn_robe_side_l_05, bone2Name=dyn_robe_side_l_04, swing1LimitDegrees=12, swing2LimitDegrees=12, twistLimitDegrees=12}
 {bone1Name=dyn_robe_front_r_03, bone2Name=dyn_robe_front_r_02, swing1LimitDegrees=12, swing2LimitDegrees=12, twistLimitDegrees=12}
 {bone1Name=dyn_robe_front_r_05, bone2Name=dyn_robe_front_r_04, swing1LimitDegrees=12, swing2LimitDegrees=12, twistLimitDegrees=12}
 {bone1Name=dyn_robe_rear_r_03, bone2Name=dyn_robe_rear_r_02, swing1LimitDegrees=12, swing2LimitDegrees=12, twistLimitDegrees=12}
 {bone1Name=dyn_robe_rear_r_05, bone2Name=dyn_robe_rear_r_04, swing1LimitDegrees=12, swing2LimitDegrees=12, twistLimitDegrees=12}
 {bone1Name=dyn_robe_side_r_03, bone2Name=dyn_robe_side_r_02, swing1LimitDegrees=12, swing2LimitDegrees=12, twistLimitDegrees=12}
 {bone1Name=dyn_robe_side_r_05, bone2Name=dyn_robe_side_r_04, swing1LimitDegrees=12, swing2LimitDegrees=12, twistLimitDegrees=12}
 {
  bone1Name=deform_thigh_r
  bone2Name=pelvis
  swing1LimitDegrees=45
  swing1Motion=Free
  swing2LimitDegrees=45
  swing2Motion=Free
  twistLimitDegrees=45
  twistMotion=Free
 }
 {
  bone1Name=deform_thigh_l
  bone2Name=pelvis
  swing1LimitDegrees=45
  swing1Motion=Free
  swing2LimitDegrees=45
  swing2Motion=Free
  twistLimitDegrees=45
  twistMotion=Free
 }
 {
  bone1Name=dyn_robe_front_l_knee_base
  bone2Name=deform_thigh_l
  swing1LimitDegrees=45
  swing1Motion=Free
  swing2LimitDegrees=45
  swing2Motion=Free
  twistLimitDegrees=45
  twistMotion=Free
 }
 {
  bone1Name=dyn_robe_rear_l_knee_base
  bone2Name=deform_thigh_l
  swing1LimitDegrees=45
  swing1Motion=Free
  swing2LimitDegrees=45
  swing2Motion=Free
  twistLimitDegrees=45
  twistMotion=Free
 }
 {
  bone1Name=dyn_robe_side_l_knee_base
  bone2Name=deform_thigh_l
  swing1LimitDegrees=45
  swing1Motion=Free
  swing2LimitDegrees=45
  swing2Motion=Free
  twistLimitDegrees=45
  twistMotion=Free
 }
 {
  bone1Name=dyn_robe_front_r_knee_base
  bone2Name=deform_thigh_r
  swing1LimitDegrees=45
  swing1Motion=Free
  swing2LimitDegrees=45
  swing2Motion=Free
  twistLimitDegrees=45
  twistMotion=Free
 }
 {
  bone1Name=dyn_robe_rear_r_knee_base
  bone2Name=deform_thigh_r
  swing1LimitDegrees=45
  swing1Motion=Free
  swing2LimitDegrees=45
  swing2Motion=Free
  twistLimitDegrees=45
  twistMotion=Free
 }
 {
  bone1Name=dyn_robe_side_r_knee_base
  bone2Name=deform_thigh_r
  swing1LimitDegrees=45
  swing1Motion=Free
  swing2LimitDegrees=45
  swing2Motion=Free
  twistLimitDegrees=45
  twistMotion=Free
 }
 {
  bone1Name=dyn_robe_front_l_knee_base
  bone2Name=dyn_robe_front_l_02
  swing1LimitDegrees=45
  swing1Motion=Free
  swing2LimitDegrees=45
  swing2Motion=Free
  twistLimitDegrees=45
  twistMotion=Free
 }
 {
  bone1Name=dyn_robe_rear_l_knee_base
  bone2Name=dyn_robe_rear_l_02
  swing1LimitDegrees=45
  swing1Motion=Free
  swing2LimitDegrees=45
  swing2Motion=Free
  twistLimitDegrees=45
  twistMotion=Free
 }
 {
  bone1Name=dyn_robe_side_l_knee_base
  bone2Name=dyn_robe_side_l_02
  swing1LimitDegrees=45
  swing1Motion=Free
  swing2LimitDegrees=45
  swing2Motion=Free
  twistLimitDegrees=45
  twistMotion=Free
 }
 {
  bone1Name=dyn_robe_front_r_knee_base
  bone2Name=dyn_robe_front_r_02
  swing1LimitDegrees=45
  swing1Motion=Free
  swing2LimitDegrees=45
  swing2Motion=Free
  twistLimitDegrees=45
  twistMotion=Free
 }
 {
  bone1Name=dyn_robe_rear_r_knee_base
  bone2Name=dyn_robe_rear_r_02
  swing1LimitDegrees=45
  swing1Motion=Free
  swing2LimitDegrees=45
  swing2Motion=Free
  twistLimitDegrees=45
  twistMotion=Free
 }
 {
  bone1Name=dyn_robe_side_r_knee_base
  bone2Name=dyn_robe_side_r_02
  swing1LimitDegrees=45
  swing1Motion=Free
  swing2LimitDegrees=45
  swing2Motion=Free
  twistLimitDegrees=45
  twistMotion=Free
 }
]
