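# Post-process-style AnimBlueprint for the M_MED_FlickLedge player body cosmetic, adding secondary motion to its robe.
# It copies pose and curves from the attached parent mesh, then runs the body's Control Rig.
# Serial single-bone Anim Dynamics nodes then simulate front, side and rear robe bones on both sides with stiff angular springs and tight limits.
# A chain Anim Dynamics node and RigidBody node are unconnected, and the update event is empty.
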AnimBlueprint /BRCosmetics/Characters/Player/Male/Medium/Bodies/M_MED_FlickLedge/Meshes/Simple/M_MED_FlickLedge_AnimBP
target_skeleton=&/M_MED_FlickLedge_Skeleton
graphs [
 AnimGraph {nodes [
   "Anim Dynamics - Chain: dyn_robe_front_l_01 -> dyn_robe_front_l_05"(Alpha=0.5, ExternalForce="0,0,0") {
    angularDampingOverride=1
    angularSpringConstant=5
    bAngularSpring=true
    bChain=true
    bGravityOverrideInSimSpace=true
    bOverrideAngularDamping=true
    bUseSphericalLimits=true
    gravityScale=0.3
    numSolverIterationsPostUpdate=8
    numSolverIterationsPreUpdate=32
    simulationSpace=RootRelative
    alphaBoolBlend {blendOption=Linear}
    boundBone {boneName=dyn_robe_front_l_01}
    chainEnd {boneName=dyn_robe_front_l_05}
    componentAppliedLinearAccClamp=<1000 1000 1000>
    componentLinearAccScale=<1 1 1>
    componentLinearVelScale=<1 1 1>
    physicsBodyDefinitions [
     {
      sphereCollisionRadius=0
      boundBone {boneName=dyn_robe_front_l_01}
      boxExtents=<18 18 30>
      constraintSetup {angularLimitsMax=<10 35 10>, angularLimitsMin=<-10 -20 -35>, angularTarget=<0 0 1>}
      localJointOffset=<5 0 0>
     }
     {
      sphereCollisionRadius=0
      boundBone {boneName=dyn_robe_front_l_02}
      boxExtents=<18 18 30>
      constraintSetup {angularLimitsMax=<10 35 10>, angularLimitsMin=<-10 -20 -35>, angularTarget=<0 0 1>}
      localJointOffset=<5 0 0>
     }
     {
      sphereCollisionRadius=0
      boundBone {boneName=dyn_robe_front_l_03}
      boxExtents=<18 18 30>
      constraintSetup {angularLimitsMax=<10 35 10>, angularLimitsMin=<-10 -20 -35>, angularTarget=<0 0 1>}
      localJointOffset=<5 0 0>
     }
     {
      sphereCollisionRadius=0
      boundBone {boneName=dyn_robe_front_l_04}
      boxExtents=<18 18 30>
      constraintSetup {angularLimitsMax=<10 35 10>, angularLimitsMin=<-10 -20 -35>, angularTarget=<0 0 1>}
      localJointOffset=<5 0 0>
     }
     {
      sphereCollisionRadius=0
      boundBone {boneName=dyn_robe_front_l_05}
      boxExtents=<18 18 30>
      constraintSetup {angularLimitsMax=<10 35 10>, angularLimitsMin=<-10 -20 -35>, angularTarget=<0 0 1>}
      localJointOffset=<5 0 0>
     }
    ]
    relativeSpaceBone {boneName=thigh_l}
    sphericalLimits [
     {limitRadius=13, drivingBone {boneName=calf_l}, sphereLocalOffset=<0 0 0>}
     {limitRadius=13, drivingBone {boneName=calf_l}, sphereLocalOffset=<-25 0 0>}
     {limitRadius=13, drivingBone {boneName=calf_l}, sphereLocalOffset=<-40 0 0>}
    ]
   }
   RigidBody() {
    overridePhysicsAsset=&/M_MED_FlickLedge_Physics
    simulationSpace=BaseBoneSpace
    baseBoneRef {boneName=pelvis}
    simSpaceSettings {maxAngularAcceleration=30, maxAngularVelocity=300, maxLinearAcceleration=30, maxLinearVelocity=770, worldAlpha=1}
   }
   n1 = Control Rig(Source=Copy Pose From Mesh() {bCopyCurves=true, bUseAttachedParent=true}) {controlRigAssetReference {blueprintRigClass=&/M_MED_FlickLedge_CtrlRig.M_MED_FlickLedge_CtrlRig_C}}
   n2 = Anim Dynamics - Bone: dyn_robe_front_l_01(ComponentPose=Local To Component(LocalPose=@n1), Alpha=1, ExternalForce="0,0,0") {
    angularDampingOverride=1
    angularSpringConstant=10000
    bAngularSpring=true
    bOverrideAngularDamping=true
    bUseGravityOverride=true
    gravityScale=0.35
    simulationSpace=RootRelative
    boundBone {boneName=dyn_robe_front_l_01}
    componentAppliedLinearAccClamp=<0 0 0>
    gravityOverride {z=-980}
    physicsBodyDefinitions [{
      boundBone {boneName=dyn_robe_front_l_01}
      boxExtents=<30 30 30>
      constraintSetup {angularLimitsMax=<15 15 0>, angularLimitsMin=<-15 -15 -15>, angularTarget=<1 0 0>}
      localJointOffset=<3 0 0>
     }]
    simSpaceSettings {maxAngularAcceleration=10, maxAngularVelocity=300, simSpaceAngularAlpha=1}
   }
   n3 = Anim Dynamics - Bone: dyn_robe_front_l_02(ComponentPose=@n2, Alpha=1, ExternalForce="0,0,0") {
    angularDampingOverride=1
    angularSpringConstant=10000
    bAngularSpring=true
    bOverrideAngularDamping=true
    bUseGravityOverride=true
    gravityScale=0.35
    simulationSpace=RootRelative
    boundBone {boneName=dyn_robe_front_l_02}
    componentAppliedLinearAccClamp=<0 0 0>
    gravityOverride {z=-980}
    physicsBodyDefinitions [{
      boundBone {boneName=dyn_robe_front_l_02}
      boxExtents=<30 30 30>
      constraintSetup {angularLimitsMax=<15 15 0>, angularLimitsMin=<-15 -15 -15>, angularTarget=<1 0 0>}
      localJointOffset=<3 0 0>
     }]
    simSpaceSettings {maxAngularAcceleration=10, maxAngularVelocity=300, simSpaceAngularAlpha=1}
   }
   n4 = Anim Dynamics - Bone: dyn_robe_front_l_03(ComponentPose=@n3, Alpha=1, ExternalForce="0,0,0") {
    angularDampingOverride=1
    angularSpringConstant=10000
    bAngularSpring=true
    bOverrideAngularDamping=true
    bUseGravityOverride=true
    gravityScale=0.35
    simulationSpace=RootRelative
    boundBone {boneName=dyn_robe_front_l_03}
    componentAppliedLinearAccClamp=<0 0 0>
    gravityOverride {z=-980}
    physicsBodyDefinitions [{
      boundBone {boneName=dyn_robe_front_l_03}
      boxExtents=<30 30 30>
      constraintSetup {angularLimitsMax=<15 15 0>, angularLimitsMin=<-15 -15 -15>, angularTarget=<1 0 0>}
      localJointOffset=<3 0 0>
     }]
    simSpaceSettings {maxAngularAcceleration=10, maxAngularVelocity=300, simSpaceAngularAlpha=1}
   }
   n5 = Anim Dynamics - Bone: dyn_robe_front_l_04(ComponentPose=@n4, Alpha=1, ExternalForce="0,0,0") {
    angularDampingOverride=1
    angularSpringConstant=10000
    bAngularSpring=true
    bOverrideAngularDamping=true
    bUseGravityOverride=true
    gravityScale=0.35
    simulationSpace=RootRelative
    boundBone {boneName=dyn_robe_front_l_04}
    componentAppliedLinearAccClamp=<0 0 0>
    gravityOverride {z=-980}
    physicsBodyDefinitions [{
      boundBone {boneName=dyn_robe_front_l_04}
      boxExtents=<30 30 30>
      constraintSetup {angularLimitsMax=<15 15 0>, angularLimitsMin=<-15 -15 -15>, angularTarget=<1 0 0>}
      localJointOffset=<3 0 0>
     }]
    simSpaceSettings {maxAngularAcceleration=10, maxAngularVelocity=300, simSpaceAngularAlpha=1}
   }
   n6 = Anim Dynamics - Bone: dyn_robe_front_l_05(ComponentPose=@n5, Alpha=1, ExternalForce="0,0,0") {
    angularDampingOverride=1
    angularSpringConstant=10000
    bAngularSpring=true
    bOverrideAngularDamping=true
    bUseGravityOverride=true
    gravityScale=0.35
    simulationSpace=RootRelative
    boundBone {boneName=dyn_robe_front_l_05}
    componentAppliedLinearAccClamp=<0 0 0>
    gravityOverride {z=-980}
    physicsBodyDefinitions [{
      boundBone {boneName=dyn_robe_front_l_05}
      boxExtents=<30 30 30>
      constraintSetup {angularLimitsMax=<15 15 0>, angularLimitsMin=<-15 -15 -15>, angularTarget=<1 0 0>}
      localJointOffset=<3 0 0>
     }]
    simSpaceSettings {maxAngularAcceleration=10, maxAngularVelocity=300, simSpaceAngularAlpha=1}
   }
   n7 = Anim Dynamics - Bone: dyn_robe_side_l_01(ComponentPose=@n6, Alpha=1, ExternalForce="0,0,0") {
    angularDampingOverride=1
    angularSpringConstant=10000
    bAngularSpring=true
    bOverrideAngularDamping=true
    bUseGravityOverride=true
    gravityScale=0.35
    simulationSpace=RootRelative
    boundBone {boneName=dyn_robe_side_l_01}
    componentAppliedLinearAccClamp=<0 0 0>
    gravityOverride {z=-980}
    physicsBodyDefinitions [{
      boundBone {boneName=dyn_robe_side_l_01}
      boxExtents=<30 30 30>
      constraintSetup {angularLimitsMax=<5 5 0>, angularLimitsMin=<-5 -5 -15>, angularTarget=<1 0 0>}
      localJointOffset=<3 0 0>
     }]
    simSpaceSettings {maxAngularAcceleration=10, maxAngularVelocity=300, simSpaceAngularAlpha=1}
   }
   n8 = Anim Dynamics - Bone: dyn_robe_side_l_02(ComponentPose=@n7, Alpha=1, ExternalForce="0,0,0") {
    angularDampingOverride=1
    angularSpringConstant=10000
    bAngularSpring=true
    bOverrideAngularDamping=true
    bUseGravityOverride=true
    gravityScale=0.35
    simulationSpace=RootRelative
    boundBone {boneName=dyn_robe_side_l_02}
    componentAppliedLinearAccClamp=<0 0 0>
    gravityOverride {z=-980}
    physicsBodyDefinitions [{
      boundBone {boneName=dyn_robe_side_l_02}
      boxExtents=<30 30 30>
      constraintSetup {angularLimitsMax=<5 5 0>, angularLimitsMin=<-5 -5 -15>, angularTarget=<1 0 0>}
      localJointOffset=<3 0 0>
     }]
    simSpaceSettings {maxAngularAcceleration=10, maxAngularVelocity=300, simSpaceAngularAlpha=1}
   }
   n9 = Anim Dynamics - Bone: dyn_robe_side_l_03(ComponentPose=@n8, Alpha=1, ExternalForce="0,0,0") {
    angularDampingOverride=1
    angularSpringConstant=10000
    bAngularSpring=true
    bOverrideAngularDamping=true
    bUseGravityOverride=true
    gravityScale=0.35
    simulationSpace=RootRelative
    boundBone {boneName=dyn_robe_side_l_03}
    componentAppliedLinearAccClamp=<0 0 0>
    gravityOverride {z=-980}
    physicsBodyDefinitions [{
      boundBone {boneName=dyn_robe_side_l_03}
      boxExtents=<30 30 30>
      constraintSetup {angularLimitsMax=<5 5 0>, angularLimitsMin=<-5 -5 -15>, angularTarget=<1 0 0>}
      localJointOffset=<3 0 0>
     }]
    simSpaceSettings {maxAngularAcceleration=10, maxAngularVelocity=300, simSpaceAngularAlpha=1}
   }
   n10 = Anim Dynamics - Bone: dyn_robe_side_l_04(ComponentPose=@n9, Alpha=1, ExternalForce="0,0,0") {
    angularDampingOverride=1
    angularSpringConstant=10000
    bAngularSpring=true
    bOverrideAngularDamping=true
    bUseGravityOverride=true
    gravityScale=0.35
    simulationSpace=RootRelative
    boundBone {boneName=dyn_robe_side_l_04}
    componentAppliedLinearAccClamp=<0 0 0>
    gravityOverride {z=-980}
    physicsBodyDefinitions [{
      boundBone {boneName=dyn_robe_side_l_04}
      boxExtents=<30 30 30>
      constraintSetup {angularLimitsMax=<5 5 0>, angularLimitsMin=<-5 -5 -15>, angularTarget=<1 0 0>}
      localJointOffset=<3 0 0>
     }]
    simSpaceSettings {maxAngularAcceleration=10, maxAngularVelocity=300, simSpaceAngularAlpha=1}
   }
   n11 = Anim Dynamics - Bone: dyn_robe_side_l_05(ComponentPose=@n10, Alpha=1, ExternalForce="0,0,0") {
    angularDampingOverride=1
    angularSpringConstant=10000
    bAngularSpring=true
    bOverrideAngularDamping=true
    bUseGravityOverride=true
    gravityScale=0.35
    simulationSpace=RootRelative
    boundBone {boneName=dyn_robe_side_l_05}
    componentAppliedLinearAccClamp=<0 0 0>
    gravityOverride {z=-980}
    physicsBodyDefinitions [{
      boundBone {boneName=dyn_robe_side_l_05}
      boxExtents=<30 30 30>
      constraintSetup {angularLimitsMax=<5 5 0>, angularLimitsMin=<-5 -5 -15>, angularTarget=<1 0 0>}
      localJointOffset=<3 0 0>
     }]
    simSpaceSettings {maxAngularAcceleration=10, maxAngularVelocity=300, simSpaceAngularAlpha=1}
   }
   n12 = Anim Dynamics - Bone: dyn_robe_rear_l_01(ComponentPose=@n11, Alpha=1, ExternalForce="0,0,0") {
    angularDampingOverride=1
    angularSpringConstant=10000
    bAngularSpring=true
    bOverrideAngularDamping=true
    bUseGravityOverride=true
    gravityScale=0.35
    simulationSpace=RootRelative
    boundBone {boneName=dyn_robe_rear_l_01}
    componentAppliedLinearAccClamp=<0 0 0>
    gravityOverride {z=-980}
    physicsBodyDefinitions [{
      boundBone {boneName=dyn_robe_rear_l_01}
      boxExtents=<30 30 30>
      constraintSetup {angularLimitsMax=<5 5 0>, angularLimitsMin=<-5 -5 -30>, angularTarget=<1 0 0>}
      localJointOffset=<3 0 0>
     }]
    simSpaceSettings {maxAngularAcceleration=10, maxAngularVelocity=300, simSpaceAngularAlpha=1}
   }
   n13 = Anim Dynamics - Bone: dyn_robe_rear_l_02(ComponentPose=@n12, Alpha=1, ExternalForce="0,0,0") {
    angularDampingOverride=1
    angularSpringConstant=10000
    bAngularSpring=true
    bOverrideAngularDamping=true
    bUseGravityOverride=true
    gravityScale=0.35
    simulationSpace=RootRelative
    boundBone {boneName=dyn_robe_rear_l_02}
    componentAppliedLinearAccClamp=<0 0 0>
    gravityOverride {z=-980}
    physicsBodyDefinitions [{
      boundBone {boneName=dyn_robe_rear_l_02}
      boxExtents=<30 30 30>
      constraintSetup {angularLimitsMax=<5 5 0>, angularLimitsMin=<-5 -5 -30>, angularTarget=<1 0 0>}
      localJointOffset=<3 0 0>
     }]
    simSpaceSettings {maxAngularAcceleration=10, maxAngularVelocity=300, simSpaceAngularAlpha=1}
   }
   n14 = Anim Dynamics - Bone: dyn_robe_rear_l_03(ComponentPose=@n13, Alpha=1, ExternalForce="0,0,0") {
    angularDampingOverride=1
    angularSpringConstant=10000
    bAngularSpring=true
    bOverrideAngularDamping=true
    bUseGravityOverride=true
    gravityScale=0.35
    simulationSpace=RootRelative
    boundBone {boneName=dyn_robe_rear_l_03}
    componentAppliedLinearAccClamp=<0 0 0>
    gravityOverride {z=-980}
    physicsBodyDefinitions [{
      boundBone {boneName=dyn_robe_rear_l_03}
      boxExtents=<30 30 30>
      constraintSetup {angularLimitsMax=<5 5 0>, angularLimitsMin=<-5 -5 -30>, angularTarget=<1 0 0>}
      localJointOffset=<3 0 0>
     }]
    simSpaceSettings {maxAngularAcceleration=10, maxAngularVelocity=300, simSpaceAngularAlpha=1}
   }
   n15 = Anim Dynamics - Bone: dyn_robe_rear_l_04(ComponentPose=@n14, Alpha=1, ExternalForce="0,0,0") {
    angularDampingOverride=1
    angularSpringConstant=10000
    bAngularSpring=true
    bOverrideAngularDamping=true
    bUseGravityOverride=true
    gravityScale=0.35
    simulationSpace=RootRelative
    boundBone {boneName=dyn_robe_rear_l_04}
    componentAppliedLinearAccClamp=<0 0 0>
    gravityOverride {z=-980}
    physicsBodyDefinitions [{
      boundBone {boneName=dyn_robe_rear_l_04}
      boxExtents=<30 30 30>
      constraintSetup {angularLimitsMax=<5 5 0>, angularLimitsMin=<-5 -5 -30>, angularTarget=<1 0 0>}
      localJointOffset=<3 0 0>
     }]
    simSpaceSettings {maxAngularAcceleration=10, maxAngularVelocity=300, simSpaceAngularAlpha=1}
   }
   n16 = Anim Dynamics - Bone: dyn_robe_rear_l_05(ComponentPose=@n15, Alpha=1, ExternalForce="0,0,0") {
    angularDampingOverride=1
    angularSpringConstant=10000
    bAngularSpring=true
    bOverrideAngularDamping=true
    bUseGravityOverride=true
    gravityScale=0.35
    simulationSpace=RootRelative
    boundBone {boneName=dyn_robe_rear_l_05}
    componentAppliedLinearAccClamp=<0 0 0>
    gravityOverride {z=-980}
    physicsBodyDefinitions [{
      boundBone {boneName=dyn_robe_rear_l_05}
      boxExtents=<30 30 30>
      constraintSetup {angularLimitsMax=<5 5 0>, angularLimitsMin=<-5 -5 -30>, angularTarget=<1 0 0>}
      localJointOffset=<3 0 0>
     }]
    simSpaceSettings {maxAngularAcceleration=10, maxAngularVelocity=300, simSpaceAngularAlpha=1}
   }
   n17 = Anim Dynamics - Bone: dyn_robe_front_r_01(ComponentPose=@n16, Alpha=1, ExternalForce="0,0,0") {
    angularDampingOverride=1
    angularSpringConstant=10000
    bAngularSpring=true
    bOverrideAngularDamping=true
    bUseGravityOverride=true
    gravityScale=0.35
    simulationSpace=RootRelative
    boundBone {boneName=dyn_robe_front_r_01}
    componentAppliedLinearAccClamp=<0 0 0>
    gravityOverride {z=-980}
    physicsBodyDefinitions [{
      boundBone {boneName=dyn_robe_front_r_01}
      boxExtents=<30 30 30>
      constraintSetup {angularLimitsMax=<15 15 0>, angularLimitsMin=<-15 -15 -15>, angularTarget=<1 0 0>}
      localJointOffset=<-3 0 0>
     }]
    simSpaceSettings {maxAngularAcceleration=10, maxAngularVelocity=300, simSpaceAngularAlpha=1}
   }
   n18 = Anim Dynamics - Bone: dyn_robe_front_r_02(ComponentPose=@n17, Alpha=1, ExternalForce="0,0,0") {
    angularDampingOverride=1
    angularSpringConstant=10000
    bAngularSpring=true
    bOverrideAngularDamping=true
    bUseGravityOverride=true
    gravityScale=0.35
    simulationSpace=RootRelative
    boundBone {boneName=dyn_robe_front_r_02}
    componentAppliedLinearAccClamp=<0 0 0>
    gravityOverride {z=-980}
    physicsBodyDefinitions [{
      boundBone {boneName=dyn_robe_front_r_02}
      boxExtents=<30 30 30>
      constraintSetup {angularLimitsMax=<15 15 0>, angularLimitsMin=<-15 -15 -15>, angularTarget=<1 0 0>}
      localJointOffset=<-3 0 0>
     }]
    simSpaceSettings {maxAngularAcceleration=10, maxAngularVelocity=300, simSpaceAngularAlpha=1}
   }
   n19 = Anim Dynamics - Bone: dyn_robe_front_r_03(ComponentPose=@n18, Alpha=1, ExternalForce="0,0,0") {
    angularDampingOverride=1
    angularSpringConstant=10000
    bAngularSpring=true
    bOverrideAngularDamping=true
    bUseGravityOverride=true
    gravityScale=0.35
    simulationSpace=RootRelative
    boundBone {boneName=dyn_robe_front_r_03}
    componentAppliedLinearAccClamp=<0 0 0>
    gravityOverride {z=-980}
    physicsBodyDefinitions [{
      boundBone {boneName=dyn_robe_front_r_03}
      boxExtents=<30 30 30>
      constraintSetup {angularLimitsMax=<15 15 0>, angularLimitsMin=<-15 -15 -15>, angularTarget=<1 0 0>}
      localJointOffset=<-3 0 0>
     }]
    simSpaceSettings {maxAngularAcceleration=10, maxAngularVelocity=300, simSpaceAngularAlpha=1}
   }
   n20 = Anim Dynamics - Bone: dyn_robe_front_r_04(ComponentPose=@n19, Alpha=1, ExternalForce="0,0,0") {
    angularDampingOverride=1
    angularSpringConstant=10000
    bAngularSpring=true
    bOverrideAngularDamping=true
    bUseGravityOverride=true
    gravityScale=0.35
    simulationSpace=RootRelative
    boundBone {boneName=dyn_robe_front_r_04}
    componentAppliedLinearAccClamp=<0 0 0>
    gravityOverride {z=-980}
    physicsBodyDefinitions [{
      boundBone {boneName=dyn_robe_front_r_04}
      boxExtents=<30 30 30>
      constraintSetup {angularLimitsMax=<15 15 0>, angularLimitsMin=<-15 -15 -15>, angularTarget=<1 0 0>}
      localJointOffset=<-3 0 0>
     }]
    simSpaceSettings {maxAngularAcceleration=10, maxAngularVelocity=300, simSpaceAngularAlpha=1}
   }
   n21 = Anim Dynamics - Bone: dyn_robe_front_r_05(ComponentPose=@n20, Alpha=1, ExternalForce="0,0,0") {
    angularDampingOverride=1
    angularSpringConstant=10000
    bAngularSpring=true
    bOverrideAngularDamping=true
    bUseGravityOverride=true
    gravityScale=0.35
    simulationSpace=RootRelative
    boundBone {boneName=dyn_robe_front_r_05}
    componentAppliedLinearAccClamp=<0 0 0>
    gravityOverride {z=-980}
    physicsBodyDefinitions [{
      boundBone {boneName=dyn_robe_front_r_05}
      boxExtents=<30 30 30>
      constraintSetup {angularLimitsMax=<15 15 0>, angularLimitsMin=<-15 -15 -15>, angularTarget=<1 0 0>}
      localJointOffset=<-3 0 0>
     }]
    simSpaceSettings {maxAngularAcceleration=10, maxAngularVelocity=300, simSpaceAngularAlpha=1}
   }
   n22 = Anim Dynamics - Bone: dyn_robe_side_r_01(ComponentPose=@n21, Alpha=1, ExternalForce="0,0,0") {
    angularDampingOverride=1
    angularSpringConstant=10000
    bAngularSpring=true
    bOverrideAngularDamping=true
    bUseGravityOverride=true
    gravityScale=0.35
    simulationSpace=RootRelative
    boundBone {boneName=dyn_robe_side_r_01}
    componentAppliedLinearAccClamp=<0 0 0>
    gravityOverride {z=-980}
    physicsBodyDefinitions [{
      boundBone {boneName=dyn_robe_side_r_01}
      boxExtents=<30 30 30>
      constraintSetup {angularLimitsMax=<5 5 0>, angularLimitsMin=<-5 -5 -15>, angularTarget=<1 0 0>}
      localJointOffset=<-3 0 0>
     }]
    simSpaceSettings {maxAngularAcceleration=10, maxAngularVelocity=300, simSpaceAngularAlpha=1}
   }
   n23 = Anim Dynamics - Bone: dyn_robe_side_r_02(ComponentPose=@n22, Alpha=1, ExternalForce="0,0,0") {
    angularDampingOverride=1
    angularSpringConstant=10000
    bAngularSpring=true
    bOverrideAngularDamping=true
    bUseGravityOverride=true
    gravityScale=0.35
    simulationSpace=RootRelative
    boundBone {boneName=dyn_robe_side_r_02}
    componentAppliedLinearAccClamp=<0 0 0>
    gravityOverride {z=-980}
    physicsBodyDefinitions [{
      boundBone {boneName=dyn_robe_side_r_02}
      boxExtents=<30 30 30>
      constraintSetup {angularLimitsMax=<5 5 0>, angularLimitsMin=<-5 -5 -15>, angularTarget=<1 0 0>}
      localJointOffset=<-3 0 0>
     }]
    simSpaceSettings {maxAngularAcceleration=10, maxAngularVelocity=300, simSpaceAngularAlpha=1}
   }
   n24 = Anim Dynamics - Bone: dyn_robe_side_r_03(ComponentPose=@n23, Alpha=1, ExternalForce="0,0,0") {
    angularDampingOverride=1
    angularSpringConstant=10000
    bAngularSpring=true
    bOverrideAngularDamping=true
    bUseGravityOverride=true
    gravityScale=0.35
    simulationSpace=RootRelative
    boundBone {boneName=dyn_robe_side_r_03}
    componentAppliedLinearAccClamp=<0 0 0>
    gravityOverride {z=-980}
    physicsBodyDefinitions [{
      boundBone {boneName=dyn_robe_side_r_03}
      boxExtents=<30 30 30>
      constraintSetup {angularLimitsMax=<5 5 0>, angularLimitsMin=<-5 -5 -15>, angularTarget=<1 0 0>}
      localJointOffset=<-3 0 0>
     }]
    simSpaceSettings {maxAngularAcceleration=10, maxAngularVelocity=300, simSpaceAngularAlpha=1}
   }
   n25 = Anim Dynamics - Bone: dyn_robe_side_r_04(ComponentPose=@n24, Alpha=1, ExternalForce="0,0,0") {
    angularDampingOverride=1
    angularSpringConstant=10000
    bAngularSpring=true
    bOverrideAngularDamping=true
    bUseGravityOverride=true
    gravityScale=0.35
    simulationSpace=RootRelative
    boundBone {boneName=dyn_robe_side_r_04}
    componentAppliedLinearAccClamp=<0 0 0>
    gravityOverride {z=-980}
    physicsBodyDefinitions [{
      boundBone {boneName=dyn_robe_side_r_04}
      boxExtents=<30 30 30>
      constraintSetup {angularLimitsMax=<5 5 0>, angularLimitsMin=<-5 -5 -15>, angularTarget=<1 0 0>}
      localJointOffset=<-3 0 0>
     }]
    simSpaceSettings {maxAngularAcceleration=10, maxAngularVelocity=300, simSpaceAngularAlpha=1}
   }
   n26 = Anim Dynamics - Bone: dyn_robe_side_r_05(ComponentPose=@n25, Alpha=1, ExternalForce="0,0,0") {
    angularDampingOverride=1
    angularSpringConstant=10000
    bAngularSpring=true
    bOverrideAngularDamping=true
    bUseGravityOverride=true
    gravityScale=0.35
    simulationSpace=RootRelative
    boundBone {boneName=dyn_robe_side_r_05}
    componentAppliedLinearAccClamp=<0 0 0>
    gravityOverride {z=-980}
    physicsBodyDefinitions [{
      boundBone {boneName=dyn_robe_side_r_05}
      boxExtents=<30 30 30>
      constraintSetup {angularLimitsMax=<5 5 0>, angularLimitsMin=<-5 -5 -15>, angularTarget=<1 0 0>}
      localJointOffset=<-3 0 0>
     }]
    simSpaceSettings {maxAngularAcceleration=10, maxAngularVelocity=300, simSpaceAngularAlpha=1}
   }
   n27 = Anim Dynamics - Bone: dyn_robe_rear_r_01(ComponentPose=@n26, Alpha=1, ExternalForce="0,0,0") {
    angularDampingOverride=1
    angularSpringConstant=10000
    bAngularSpring=true
    bOverrideAngularDamping=true
    bUseGravityOverride=true
    gravityScale=0.35
    simulationSpace=RootRelative
    boundBone {boneName=dyn_robe_rear_r_01}
    componentAppliedLinearAccClamp=<0 0 0>
    gravityOverride {z=-980}
    physicsBodyDefinitions [{
      boundBone {boneName=dyn_robe_rear_r_01}
      boxExtents=<30 30 30>
      constraintSetup {angularLimitsMax=<5 5 0>, angularLimitsMin=<-5 -5 -30>, angularTarget=<1 0 0>}
      localJointOffset=<-3 0 0>
     }]
    simSpaceSettings {maxAngularAcceleration=10, maxAngularVelocity=300, simSpaceAngularAlpha=1}
   }
   n28 = Anim Dynamics - Bone: dyn_robe_rear_r_02(ComponentPose=@n27, Alpha=1, ExternalForce="0,0,0") {
    angularDampingOverride=1
    angularSpringConstant=10000
    bAngularSpring=true
    bOverrideAngularDamping=true
    bUseGravityOverride=true
    gravityScale=0.35
    simulationSpace=RootRelative
    boundBone {boneName=dyn_robe_rear_r_02}
    componentAppliedLinearAccClamp=<0 0 0>
    gravityOverride {z=-980}
    physicsBodyDefinitions [{
      boundBone {boneName=dyn_robe_rear_r_02}
      boxExtents=<30 30 30>
      constraintSetup {angularLimitsMax=<5 5 0>, angularLimitsMin=<-5 -5 -30>, angularTarget=<1 0 0>}
      localJointOffset=<-3 0 0>
     }]
    simSpaceSettings {maxAngularAcceleration=10, maxAngularVelocity=300, simSpaceAngularAlpha=1}
   }
   n29 = Anim Dynamics - Bone: dyn_robe_rear_r_03(ComponentPose=@n28, Alpha=1, ExternalForce="0,0,0") {
    angularDampingOverride=1
    angularSpringConstant=10000
    bAngularSpring=true
    bOverrideAngularDamping=true
    bUseGravityOverride=true
    gravityScale=0.35
    simulationSpace=RootRelative
    boundBone {boneName=dyn_robe_rear_r_03}
    componentAppliedLinearAccClamp=<0 0 0>
    gravityOverride {z=-980}
    physicsBodyDefinitions [{
      boundBone {boneName=dyn_robe_rear_r_03}
      boxExtents=<30 30 30>
      constraintSetup {angularLimitsMax=<5 5 0>, angularLimitsMin=<-5 -5 -30>, angularTarget=<1 0 0>}
      localJointOffset=<-3 0 0>
     }]
    simSpaceSettings {maxAngularAcceleration=10, maxAngularVelocity=300, simSpaceAngularAlpha=1}
   }
   n30 = Anim Dynamics - Bone: dyn_robe_rear_r_04(ComponentPose=@n29, Alpha=1, ExternalForce="0,0,0") {
    angularDampingOverride=1
    angularSpringConstant=10000
    bAngularSpring=true
    bOverrideAngularDamping=true
    bUseGravityOverride=true
    gravityScale=0.35
    simulationSpace=RootRelative
    boundBone {boneName=dyn_robe_rear_r_04}
    componentAppliedLinearAccClamp=<0 0 0>
    gravityOverride {z=-980}
    physicsBodyDefinitions [{
      boundBone {boneName=dyn_robe_rear_r_04}
      boxExtents=<30 30 30>
      constraintSetup {angularLimitsMax=<5 5 0>, angularLimitsMin=<-5 -5 -30>, angularTarget=<1 0 0>}
      localJointOffset=<-3 0 0>
     }]
    simSpaceSettings {maxAngularAcceleration=10, maxAngularVelocity=300, simSpaceAngularAlpha=1}
   }
   n31 = Anim Dynamics - Bone: dyn_robe_rear_r_05(ComponentPose=@n30, Alpha=1, ExternalForce="0,0,0") {
    angularDampingOverride=1
    angularSpringConstant=10000
    bAngularSpring=true
    bOverrideAngularDamping=true
    bUseGravityOverride=true
    gravityScale=0.35
    simulationSpace=RootRelative
    boundBone {boneName=dyn_robe_rear_r_05}
    componentAppliedLinearAccClamp=<0 0 0>
    gravityOverride {z=-980}
    physicsBodyDefinitions [{
      boundBone {boneName=dyn_robe_rear_r_05}
      boxExtents=<30 30 30>
      constraintSetup {angularLimitsMax=<5 5 0>, angularLimitsMin=<-5 -5 -30>, angularTarget=<1 0 0>}
      localJointOffset=<-3 0 0>
     }]
    simSpaceSettings {maxAngularAcceleration=10, maxAngularVelocity=300, simSpaceAngularAlpha=1}
   }
   Output Pose(Result=Component To Local(ComponentPose=@n31)) {layerGroup=None}
   Reroute Node(InputPin:Wildcard) -> (OutputPin:Wildcard)
   Get PawnSpeed2D()
  ]}
 EventGraph {nodes [Event BlueprintUpdateAnimation()]}
]
variables [
 Gravity Override {type=Vector}
 Gravity Override_AnimDynamics {type=Vector}
 MoveFloorCollision {type=boolean}
 MovingForward {type=boolean}
 RBAN Settings Out {type=FortRigidBodyAnimNodeParameters}
 RBAN Settings Sliding {
  type=FortRigidBodyAnimNodeParameters
  default {simSpaceSettings {
    dampingAlpha=0
    maxAngularAcceleration=10
    maxAngularVelocity=4
    maxLinearAcceleration=1000
    maxLinearVelocity=200
    velocityScaleZ=0.4
    worldAlpha=0.4
   }}
 }
]
